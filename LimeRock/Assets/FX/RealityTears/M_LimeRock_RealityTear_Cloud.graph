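# Unlit translucent, two-sided cloud material for LimeRock reality-tear FX on Niagara sprites and spline meshes.
# Panned cell-noise and storm-cloud textures, with spherized, Perlin-distorted UVs, form masks for shaded color, highlights, nebula and opacity.
# Dynamic parameter Crack drives an opening animation, which a static switch turns into an animated sphere mask for spline meshes.
# Particle random and PanAdd offset the panning time.
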Material /LimeRock/Assets/FX/RealityTears/M_LimeRock_RealityTear_Cloud
graph {
 properties {
  bAutomaticallySetUsageInEditor=false
  bUsedWithNiagaraSprites=true
  bUsedWithSplineMeshes=true
  blendMode=BLEND_Translucent
  shadingModel=MSM_Unlit
  twoSided=true
  usedShadingModels=Unlit
 }
 nodes [
  n1 = VectorParameter() {parameterName=MainTile, defaultValue=<rgba 1.7 1.7 0.02 -0.06>}
  n2 = NamedRerouteUsage()
  n3 = NamedRerouteUsage()
  n4 = Panner(Coordinate=Multiply(A=ComponentMask(None=@n1.RGB) {g=true, r=true}, B=@n2), Time=@n3, Speed=ComponentMask(None=@n1.A) {a=true, b=true})
  n5 = VectorParameter() {parameterName=DetailTilePan, defaultValue=<rgba 1.2 1.2 0.07 0.05>}
  n6 = Multiply(A=@n2, B=ComponentMask(None=@n5.RGB) {g=true, r=true})
  n7 = ComponentMask(None=@n5.A) {a=true, b=true}
  n8 = Panner(Coordinate=@n6, Time=@n3, Speed=@n7)
  n9 = TextureSample(UVs=@n8) {samplerType=SAMPLERTYPE_Normal, texture=/Game/Global/Textures/Default/Tileable_Patterns/T_StormCloudDetail01_N}
  n10 = LinearInterpolate(Alpha=NamedRerouteUsage()) {constA=0.2, constB=0.8}
  n11 = ConstantBiasScale(None=NamedRerouteUsage()) {bias=-0.5, scale=2}
  n12 = Constant()
  Add(
   A=MaterialFunctionCall(
    BaseNormal=TextureSample(UVs=@n4) {samplerType=SAMPLERTYPE_Normal, texture=/Game/Effects/Fort_Effects/Textures/Tile/T_CellNoise_Tile_N}.RGB
    AdditionalNormal=MaterialFunctionCall(Normal=@n9.RGB, Flatness=@n10) {materialFunction=&FlattenNormal}.Result
   ) {materialFunction=&BlendAngleCorrectedNormals}.Result
   B=Multiply(A=AppendVector(A=@n11, B=@n12)) {constB=0.5}
  )
  n13 = Constant() {r=0.5}
  n14 = SphereMask(A=TextureCoordinate(), B=@n13, Radius=@n13) {hardnessPercent=0}
  n15 = Power(Base=@n14) {constExponent=0.6}
  n16 = TextureSample(UVs=@n4) {samplerType=SAMPLERTYPE_LinearGrayscale, texture=/Game/Effects/Fort_Effects/Textures/Tile/T_CellNoise_Tile_M}
  n17 = Multiply(
   A=LinearInterpolate(Alpha=@n16.G) {constA=0.2}
   B=LinearInterpolate(Alpha=TextureSample(UVs=@n8) {samplerType=SAMPLERTYPE_LinearGrayscale, texture=/Game/Global/Textures/Default/Tileable_Patterns/T_StormCloudDetail01_M}.G) {constA=0.1}
  )
  n18 = Multiply(
   A=NamedRerouteDeclaration(None=@n15) {name=MainMask, nodeColor=<rgba 0 1 0.4844 1>}
   B=SmoothStep(Max=@n17, Value=LinearInterpolate(Alpha=@n14) {constA=0.02, constB=0.1}) {constMax=0, constValue=1}
  )
  n19 = Power(Base=@n18)
  n20 = Add(
   A=Multiply(A=VectorParameter() {parameterName=CloudHighlightColor, defaultValue=<rgba 2 0.0417 0.8164 1>}.RGB, B=@n19)
   B=Multiply(A=NamedRerouteUsage(), B=@n18)
  )
  n21 = LinearInterpolate(Alpha=@n15) {constA=0.04, constB=0.2}
  n22 = LinearInterpolate(
   A=VectorParameter() {parameterName=CloudColor, defaultValue=<rgba 0.2537 0.1393 0.3802 1>}.RGB
   B=VectorParameter() {parameterName=CloudShadowColor, defaultValue=<rgba 0.023 0.0397 0.07 1>}.RGB
   Alpha=Power(Base=SmoothStep(Min=@n17, Value=@n21) {constMax=0}) {constExponent=1.5}
  )
  n23 = ParticleColor()
  MaterialFunctionCall(Emissive=Multiply(A=Add(A=@n20, B=@n22), B=@n23.RGB)) {materialFunction=&MF_FX_EmissiveAdjustments}
  n24 = NamedRerouteUsage()
  n25 = Add(A=@n17, B=@n24)
  n26 = LinearInterpolate(Alpha=@n24) {constA=5, constB=0.1}
  n27 = Add(A=@n19, B=Multiply(A=Power(Base=@n25, Exp=@n26), B=@n24))
  n28 = ScalarParameter() {defaultValue=2, parameterName=OpacityHardness}
  Multiply(A=@n23.A, B=Saturate(None=Multiply(A=@n27, B=@n28)))
  n29 = ConstantBiasScale(None=TextureCoordinate()) {bias=-0.5, scale=2}
  n30 = Divide(A=@n29) {constB=1.2}
  n31 = OneMinus(None=Saturate(None=DotProduct(A=@n30, B=@n30)))
  n32 = ScalarParameter() {defaultValue=0.5, parameterName=Spherize UV}
  n33 = Divide(A=@n29, B=Add(A=Multiply(A=@n31, B=@n32) {constB=0.5}))
  n34 = TextureCoordinate()
  n35 = Constant() {r=0.5}
  n36 = VectorParameter() {parameterName=DistortTilePan, defaultValue=<rgba 1 1 -0.01 0.02>}
  n37 = Panner(
   Coordinate=Multiply(A=@n34, B=ComponentMask(None=@n36.RGB) {g=true, r=true})
   Time=NamedRerouteUsage()
   Speed=ComponentMask(None=@n36.A) {a=true, b=true}
  )
  n38 = Multiply(
   A=Multiply(A=SphereMask(A=@n34, B=@n35) {attenuationRadius=0.5, hardnessPercent=0}, B=ScalarParameter() {parameterName=UVDistortion})
   B=Subtract(A=TextureSample(UVs=@n37) {samplerType=SAMPLERTYPE_LinearColor, texture=/Game/Effects/Fort_Effects/Textures/Tile/T_PerlinRGB_Displacement}.RGB) {constB=0.5}
  ) {constB=0.1}
  NamedRerouteDeclaration(None=Add(A=ConstantBiasScale(None=@n33), B=ComponentMask(None=@n38) {g=true, r=true})) {name=DistortedUVs, nodeColor=<rgba 1 0.0098 0 1>}
  n39 = Power(Base=ComponentMask(None=TextureCoordinate()) {g=true}) {constExponent=0.2}
  n40 = DynamicParameter() {paramNames [Crack, SplineWidthMult, PanAdd, Param4]}
  n41 = SphereMask(
   A=TextureCoordinate() {coordinateIndex=2}
   B=Constant() {r=0.5}
   Radius=LinearInterpolate(A=ScalarParameter(), B=ScalarParameter() {defaultValue=1, parameterName=s}, Alpha=@n40.Crack) {constB=3}
   Hardness=LinearInterpolate(
    A=ScalarParameter() {defaultValue=0.25, parameterName=Param_2}
    B=ScalarParameter() {defaultValue=0.7, parameterName=Param_1}
    Alpha=@n40.Crack
   ) {constA=0.7}
  ) {hardnessPercent=0}
  NamedRerouteDeclaration(None=StaticSwitchParameter(True=Multiply(A=@n39, B=@n41), False=@n40.Crack) {parameterName=UsedWithSplineMesh}) {name=OpenAnim, nodeColor=<rgba 0 0.4929 1 1>}
  n42 = Time()
  n43 = ParticleRandom()
  n44 = NamedRerouteUsage()
  NamedRerouteDeclaration(None=Add(A=Add(A=@n42, B=@n43), B=Add(A=@n44, B=@n40.PanAdd))) {name=PanTime, nodeColor=<rgba 1 0 0 1>}
  n45 = Multiply(A=@n7) {constB=0.4}
  n46 = TextureSample(UVs=Multiply(A=Panner(Coordinate=@n6, Time=@n3, Speed=@n45), B=Constant2Vector() {g=1.2, r=-1.2})) {texture=/Game/Characters/Player/Male/Medium/Bodies/M_Med_Soldier_04/Skins/Galaxy/Textures/T_FN_Nebula_style2}
  n47 = ScalarParameter() {defaultValue=2, parameterName=NebulaBrightness}
  NamedRerouteDeclaration(None=Multiply(A=OneMinus(None=@n16.G), B=Multiply(A=@n46.RGB, B=@n47) {constB=3})) {name=Nebula, nodeColor=<rgba 0 0.8828 1 1>}
 ]
}
