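# Unlit translucent, two-sided material that renders a glowing reality-tear rift on Niagara ribbons, mesh particles and spline meshes.
# Noise-distorted UV sphere masks threshold tear and panning cloud textures into crack glow, clouds and nebula, with normal-based fake lighting and simpler mobile color.
# Dynamic parameters drive cracking, center opening, width and pan time.
# A spline-mesh switch changes UV orientation, opening and opacity, and adds camera-aware thickening through world position offset.
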Material /LimeRock/Assets/FX/RealityTears/M_LimeRock_RealityTear_Ribbon
graph {
 properties {
  bAutomaticallySetUsageInEditor=false
  bUsedWithNiagaraMeshParticles=true
  bUsedWithNiagaraRibbons=true
  bUsedWithSplineMeshes=true
  blendMode=BLEND_Translucent
  shadingModel=MSM_Unlit
  twoSided=true
  usedShadingModels=Unlit
 }
 nodes [
  Constant() {r=0.5}
  Constant() {r=1}
  Constant() {r=0.5}
  MaterialFunctionCall() {materialFunction=&DebugTimeSine}
  n1 = VectorParameter() {parameterName=RiftCrackColor, defaultValue=<rgba 5 1.25 1.5606 1>}
  n2 = NamedRerouteUsage()
  n3 = TextureCoordinate()
  n4 = VectorParameter() {parameterName=VarianceTilePan, defaultValue=<rgba 0.5 0.3 0.008 0.03>}
  n5 = ComponentMask(None=@n4.RGB) {g=true, r=true}
  n6 = TextureSample(UVs=Panner(Coordinate=Multiply(A=@n3, B=@n5), Speed=ComponentMask(None=@n4.A) {a=true, b=true})) {samplerType=SAMPLERTYPE_LinearGrayscale, texture=/Game/Global/Textures/Default/Tileable_Patterns/T_StormCloudDetail01_M}
  n7 = Power(Base=@n6.G) {constExponent=0.2}
  n8 = LinearInterpolate(A=ScalarParameter() {defaultValue=0.5, parameterName=UVDistortion}, Alpha=@n7) {constB=0.02}
  n9 = TextureCoordinate()
  n10 = StaticSwitchParameter(True=@n9, False=MaterialFunctionCall(XY=@n9) {materialFunction=&Swizzle}.YX) {parameterName=UsedWithSplineMesh}
  n11 = VectorParameter() {parameterName=DistortTilePan, defaultValue=<rgba 1 1 -0.01 0.02>}
  n12 = ComponentMask(None=@n11.RGB) {g=true, r=true}
  n13 = TextureSample(UVs=Panner(Coordinate=Multiply(A=@n10, B=@n12), Time=NamedRerouteUsage(), Speed=ComponentMask(None=@n11.A) {a=true, b=true})) {samplerType=SAMPLERTYPE_LinearColor, texture=/Game/Effects/Fort_Effects/Textures/Tile/T_PerlinRGB_Displacement}
  n14 = ComponentMask(None=Multiply(A=@n8, B=Subtract(A=@n13.RGB) {constB=0.5}) {constB=0.1}) {g=true, r=true}
  n15 = TextureCoordinate()
  n16 = StaticSwitchParameter(True=@n15, False=MaterialFunctionCall(XY=@n15) {materialFunction=&Swizzle}.YX) {parameterName=UsedWithSplineMesh}
  n17 = ComponentMask(None=NamedRerouteDeclaration(None=Add(A=@n14, B=@n16)) {name=DistortedUVs, nodeColor=<rgba 1 0.0098 0 1>}) {r=true}
  n18 = ScalarParameter() {defaultValue=0.3, parameterName=CenterMaskRadius}
  n19 = NamedRerouteUsage()
  n20 = NamedRerouteUsage()
  n21 = SphereMask(A=@n17, B=Constant() {r=0.5}, Radius=LinearInterpolate(B=Multiply(A=@n18, B=@n19), Alpha=@n20) {constA=-0.1}) {attenuationRadius=0.05, hardnessPercent=0}
  n22 = NamedRerouteDeclaration(None=Power(Base=Multiply(A=@n2, B=@n21)) {constExponent=8}) {name=CenterMask, nodeColor=<rgba 0 1 0.4375 1>}
  n23 = SphereMask(A=@n17, B=Constant() {r=0.5}, Radius=LinearInterpolate(Alpha=@n20) {constB=0.5}) {hardnessPercent=0}
  n24 = NamedRerouteDeclaration(None=Power(Base=Multiply(A=@n2, B=@n23)) {constExponent=0.6}) {name=MainMask, nodeColor=<rgba 0 1 0.4844 1>}
  n25 = Max(A=Add(A=Multiply(A=@n22) {constB=0.4}, B=@n24)) {constB=0.01}
  n26 = NamedRerouteUsage()
  n27 = Multiply(A=ComponentMask(None=VectorParameter() {parameterName=MainTile, defaultValue=<rgba 1 1 0.6 1>}.RGB) {g=true, r=true}, B=@n26)
  n28 = TextureSample(UVs=@n27) {samplerType=SAMPLERTYPE_LinearGrayscale, texture=&/T_RealityTear_M}
  n29 = VectorParameter() {parameterName=DetailTilePan, defaultValue=<rgba 1 2 0.07 0.06>}
  n30 = Multiply(A=@n26, B=ComponentMask(None=@n29.RGB) {g=true, r=true})
  n31 = NamedRerouteUsage()
  n32 = NamedRerouteUsage()
  n33 = LinearInterpolate(Alpha=Round(None=ComponentMask(None=@n32) {r=true})) {constA=1, constB=-1}
  n34 = Multiply(A=ComponentMask(None=@n29.A) {a=true, b=true}, B=AppendVector(A=@n33, B=Constant() {r=1}))
  n35 = Panner(Coordinate=@n30, Time=@n31, Speed=@n34)
  n36 = Multiply(
   A=LinearInterpolate(Alpha=@n28.G) {constA=0.3}
   B=LinearInterpolate(Alpha=TextureSample(UVs=@n35) {samplerType=SAMPLERTYPE_LinearGrayscale, texture=&T_StormCloudDetail01_M}.G) {constA=0.3}
  )
  n37 = Power(Base=Saturate(None=SmoothStep(Min=@n25, Value=@n36) {constMax=0, constValue=1})) {constExponent=8}
  n38 = Multiply(A=LinearInterpolate(Alpha=NamedRerouteUsage()) {constA=0.2}, B=@n37)
  n39 = VectorParameter() {parameterName=CloudHighlightColor, defaultValue=<rgba 2 0.125 0.8667 1>}
  n40 = Multiply(A=@n39.RGB, B=@n37)
  n41 = Multiply(A=NamedRerouteUsage(), B=Add(A=Multiply(A=@n1.RGB, B=@n38) {constB=0}, B=@n40))
  n42 = Multiply(
   A=LinearInterpolate(A=VectorParameter() {parameterName=NebulaColor, defaultValue=<rgba 1 0.2846 0.8553 1>}.RGB, B=@n1.RGB, Alpha=NamedRerouteUsage())
   B=NamedRerouteUsage()
  )
  n43 = VectorParameter() {parameterName=CloudShadowColor, defaultValue=<rgba 0.0262 0.0573 0.1 1>}
  n44 = Multiply(A=ScalarParameter() {defaultValue=0.25, parameterName=MinBrightness}, B=@n43.RGB) {constB=0.2}
  n45 = VectorParameter() {parameterName=CloudColor, defaultValue=<rgba 0.1224 0.1534 0.3802 1>}
  n46 = LinearInterpolate(A=@n45.RGB, B=@n39.RGB, Alpha=NamedRerouteUsage())
  n47 = NamedRerouteUsage()
  n48 = SmoothStep(Min=Max(A=@n36) {constB=0.01}, Value=Multiply(A=@n24) {constA=0.5, constB=0.3}) {constMax=0}
  n49 = LinearInterpolate(Alpha=NamedRerouteUsage()) {constA=0.5, constB=1.2}
  n50 = Power(Base=Saturate(None=Multiply(A=@n48, B=@n49)), Exp=Add(A=NamedRerouteUsage()))
  n51 = ShadingPathSwitch(
   Default=LinearInterpolate(A=Add(A=@n42, B=@n44), B=LinearInterpolate(A=@n43.RGB, B=@n46, Alpha=@n47) {constB=0}, Alpha=@n50)
   Mobile=LinearInterpolate(A=@n40, B=@n45.RGB, Alpha=@n50)
  )
  n52 = ParticleColor()
  MaterialFunctionCall(Emissive=Multiply(A=Add(A=@n41, B=@n51), B=@n52.RGB)) {materialFunction=&MF_FX_EmissiveAdjustments}
  n53 = ComponentMask(None=@n10) {r=true}
  n54 = Constant() {r=0.5}
  Multiply(
   A=NamedRerouteDeclaration(None=SphereMask(A=@n53, B=@n54) {attenuationRadius=0.5, hardnessPercent=0}) {name=FixedMask, nodeColor=<rgba 1 0.1172 0 1>}
   B=@n8
  )
  n55 = OneMinus(None=@n28.G)
  Multiply(A=NamedRerouteUsage(), B=@n55)
  n56 = TextureCoordinate()
  n57 = ScalarParameter()
  n58 = ScalarParameter() {defaultValue=1, parameterName=s}
  n59 = DynamicParameter() {paramNames [Crack, SplineWidthMult, PanAdd, CenterOpen]}
  n60 = ScalarParameter() {defaultValue=0.25, parameterName=Param_2}
  n61 = ScalarParameter() {defaultValue=0.7, parameterName=Param_1}
  n62 = Multiply(
   A=Power(Base=ComponentMask(None=@n56) {g=true}) {constExponent=0.2}
   B=SphereMask(
    A=TextureCoordinate() {coordinateIndex=2}
    B=Constant() {r=0.5}
    Radius=LinearInterpolate(A=@n57, B=@n58, Alpha=@n59.Crack) {constB=3}
    Hardness=LinearInterpolate(A=@n60, B=@n61, Alpha=@n59.Crack) {constA=0.7}
   ) {hardnessPercent=0}
  )
  NamedRerouteDeclaration(None=Multiply(A=StaticSwitchParameter(True=@n62, False=@n59.Crack) {parameterName=UsedWithSplineMesh}, B=NamedRerouteUsage())) {name=OpenAnim, nodeColor=<rgba 0 0.4929 1 1>}
  n63 = Max(A=@n7, B=Multiply(A=NamedRerouteUsage()) {constB=0.75})
  n64 = MaterialFunctionCall(In=Multiply(A=@n63, B=@n63), Contrast=Constant() {r=0.2}) {materialFunction=&CheapContrast}
  NamedRerouteDeclaration(None=MaterialFunctionCall(
    Normal=TextureSample() {samplerType=SAMPLERTYPE_Normal, texture=/Game/Global/Textures/Default/Tileable_Patterns/T_StormCloudDetail01_N}.RGB
    Flatness=LinearInterpolate(Alpha=@n64.Result) {constA=-2}
   ) {materialFunction=&FlattenNormal}.Result) {name=VarianceNormal, nodeColor=<rgba 0 0.7891 1 1>}
  NamedRerouteDeclaration(None=@n59.CenterOpen) {name=CenterOpenParam, nodeColor=<rgba 0 0.4929 1 1>}
  n65 = NamedRerouteUsage()
  n66 = NamedRerouteUsage()
  n67 = MaterialFunctionCall(
   BaseNormal=TextureSample(UVs=@n27) {samplerType=SAMPLERTYPE_Normal, texture=&/T_RealityTear_N}.RGB
   AdditionalNormal=MaterialFunctionCall(
    Normal=TextureSample(UVs=@n35) {samplerType=SAMPLERTYPE_Normal, texture=&T_StormCloudDetail01_N}.RGB
    Flatness=LinearInterpolate(Alpha=@n66) {constA=0.2, constB=0.9}
   ) {materialFunction=&FlattenNormal}.Result
  ) {materialFunction=&BlendAngleCorrectedNormals}
  n68 = NamedRerouteUsage()
  n69 = MaterialFunctionCall(X=@n33, Y=Constant(), Z=NamedRerouteUsage()) {materialFunction=&MakeFloat3}
  n70 = Saturate(None=DotProduct(A=LinearInterpolate(A=@n65, B=@n67.Result, Alpha=@n68), B=Normalize(VectorInput=@n69.Result)))
  n71 = Power(Base=@n70) {constExponent=6}
  n72 = Multiply(A=Power(Base=@n66) {constExponent=6}) {constB=0.4}
  NamedRerouteDeclaration(None=Multiply(A=NamedRerouteUsage(), B=Max(A=@n71, B=@n72))) {name=NormalHightlightMask, nodeColor=<rgba 1 0 0.4219 1>}
  NamedRerouteDeclaration(None=Power(Base=@n70) {constExponent=1}) {name=NormalShadowMask, nodeColor=<rgba 1 0 0.4219 1>}
  NamedRerouteDeclaration(None=@n64.Result) {name=VarianceMask, nodeColor=<rgba 0 1 0.4844 1>}
  n73 = NamedRerouteUsage()
  NamedRerouteDeclaration(None=Add(A=Time(), B=Add(A=@n73, B=@n59.PanAdd))) {name=PanTime, nodeColor=<rgba 1 0 0 1>}
  NamedRerouteDeclaration(None=@n59.SplineWidthMult) {name=SplineWidthParam, nodeColor=<rgba 0 0.4929 1 1>}
  n74 = Multiply(A=@n34) {constB=0.4}
  n75 = TextureSample(UVs=Multiply(A=Panner(Coordinate=@n30, Time=@n31, Speed=@n74), B=Constant2Vector() {g=0.5, r=0.9})) {texture=/Game/Characters/Player/Male/Medium/Bodies/M_Med_Soldier_04/Skins/Galaxy/Textures/T_FN_Nebula_style2}
  n76 = ScalarParameter() {defaultValue=2, parameterName=NebulaBrightness}
  NamedRerouteDeclaration(None=Multiply(A=@n55, B=Multiply(A=@n75.RGB, B=@n76) {constB=3})) {name=Nebula, nodeColor=<rgba 0 0.8828 1 1>}
  OneMinus(None=@n6.G)
  n77 = Multiply(
   A=Multiply(A=NamedRerouteUsage(), B=ScalarParameter() {defaultValue=500, parameterName=SplineWidth})
   B=MaterialFunctionCall(Distance=ScalarParameter() {defaultValue=2000, parameterName=NearCameraShrinkDistScale}) {materialFunction=&Near_Camera_Fade}.Result
  )
  StaticSwitchParameter(
   True=MaterialFunctionCall(WidthBase=@n77, WidthTip=@n77) {materialFunction=&SplineThicken}.WorldPositionOffset
   False=Constant()
  ) {parameterName=UsedWithSplineMesh}
  n78 = NamedRerouteUsage()
  n79 = Add(A=Multiply(A=@n36, B=@n78), B=Multiply(A=@n78) {constB=0.6})
  n80 = Add(A=@n37, B=Multiply(A=@n79, B=@n79))
  n81 = ScalarParameter() {defaultValue=1, parameterName=OpacityHardness}
  n82 = Multiply(A=@n52.A, B=Saturate(None=Multiply(A=@n80, B=@n81)))
  StaticSwitchParameter(True=Multiply(A=VertexColor().G, B=@n82), False=@n82) {parameterName=UsedWithSplineMesh}
 ]
}
