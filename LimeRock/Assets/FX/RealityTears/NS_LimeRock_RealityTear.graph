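# Niagara system for a reality tear effect, drawing wandering ribbon rifts trailed by drifting pink-purple stars.
# Leaders burst in waves and wobble their velocity by waveform; FollowerRibbons spawns from them, animating material parameters by age and tinting by distance.
# Stars spawns around follower particles with noise-driven drift.
# Leaders emit location events at two lifetime points with no handler shown, and the sine-path FixedRibbon emitter is disabled.
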NiagaraSystem /LimeRock/Assets/FX/RealityTears/NS_LimeRock_RealityTear
dependencies {
 usedDataInterfaces [&NiagaraDataInterfaceCurve, &NiagaraDataInterfaceParticleRead, &NiagaraDataInterfaceArrayFloat3]
 usedDynamicInputs []
 usedModules [&SystemState]
}
emitters {
 FixedRibbon {
  bEnabled=false
  emitterUpdateScript [
   &EmitterState(
    Inactive Response=::NewEnumerator1
    Life Cycle Mode=::NewEnumerator1
    Loop Behavior=::NewEnumerator1
    Loop Duration Mode=::NewEnumerator1
    UseLoopDelay=false
   )
   &SpawnBurst_Instantaneous(Spawn Count=12, Use Spawn Probability=false, UseLoopCountLimit=false)
   &SpawnBurst_Instantaneous(Spawn Count=11, Spawn Group=1, Spawn Time=0.02, Use Spawn Probability=false, UseLoopCountLimit=false) {module=SpawnBurst_Instantaneous001}
   &SpawnBurst_Instantaneous(Spawn Count=10, Spawn Group=2, Spawn Time=0.04, Use Spawn Probability=false, UseLoopCountLimit=false) {module=SpawnBurst_Instantaneous002}
  ]
  particleSpawnScript [
   /SetVariables(Particles.RibbonID=/Niagara/DynamicInputs/Meshes/MakeRibbonIDFromSpawnGroup()) {isSetParameters=true}
   /SetVariables(
    Particles.RibbonPath1:Vector3f=&Add_Vector(
     A=&Multiply_VectorByFloat(Vector=<5000 1 1>, Float=&ReturnNormalizedExecIndex())
     B=&MakeVector(
      X=0
      Y=&Waveform(
       Global Phase=&ReturnNormalizedExecIndex()
       Global Amplitude Scale=300
       "Waveform [1]"=::NewEnumerator0
       "[1] Frequency"=1.5
       "[1] Amplitude Min/Max"=<-1 1>
      )
      Z=0
     )
    )
    Particles.RibbonPath2:Vector3f=&Add_Vector(
     A=&Add_Vector(
      A=<2500 -100 0>
      B=&Multiply_VectorByFloat(Vector=<3000 -3000 500>, Float=&ReturnNormalizedExecIndex())
     )
     B=&MakeVector(
      X=0
      Y=&Waveform(
       Global Input Position=0.6
       Global Phase=&ReturnNormalizedExecIndex()
       Global Amplitude Scale=250
       "Waveform [1]"=::NewEnumerator0
       "[1] Frequency"=1.5
       "[1] Amplitude Min/Max"=<-1 1>
      )
      Z=0
     )
    )
    Particles.RibbonPath3:Vector3f=&Add_Vector(
     A=&Add_Vector(
      A=<3500 -1200 180>
      B=&Multiply_VectorByFloat(Vector=<3700 1200 500>, Float=&ReturnNormalizedExecIndex())
     )
     B=&MakeVector(
      X=0
      Y=&Waveform(
       Global Input Position=0.4
       Global Phase=&ReturnNormalizedExecIndex()
       Global Amplitude Scale=300
       "Waveform [1]"=::NewEnumerator0
       "[1] Frequency"=1
       "[1] Amplitude Min/Max"=<-1 1>
      )
      Z=0
     )
    )
   ) {isSetParameters=true}
   &InitializeParticle(
    Color=&LerpLinearColors(
     StartColor=<rgba 1 1 1 0>
     EndColor=<rgba 1 1 1 1>
     LerpFactor=&FloatFromCurve(
      FloatCurve=DataInterface() {propertyValues {Curve {
         defaultValue=1
         keys [
          {arriveTangent=26.9494, interpMode=RCIM_Cubic, leaveTangent=26.9494, tangentMode=RCTM_User, time=0, value=0}
          {arriveTangentWeight=0.1, interpMode=RCIM_Cubic, leaveTangentWeight=0.1333, tangentMode=RCTM_User, time=0.1, value=1}
          {arriveTangentWeight=0.2325, interpMode=RCIM_Cubic, leaveTangentWeight=0.1744, tangentMode=RCTM_User, time=0.9, value=1}
          {arriveTangent=-22.9197, tangentMode=RCTM_User, time=1, value=0}
         ]
        }}}
      CurveIndex=&ReturnNormalizedExecIndex()
     )
    )
    Lifetime=2
    Position Offset=&self:SelectPositionCurve(
     Index:NiagaraInt32=$Emitter.SpawnGroup
     1:Vector3f=$Particles.RibbonPath1
     2:Vector3f=$Particles.RibbonPath2
     3:Vector3f=$Particles.RibbonPath3
    )
    Ribbon Facing Mode=::NewEnumerator1
    Ribbon Width=&MaskFloatBySpawnGroup(Masked Float=900, Passthrough Float=800)
    Ribbon Width Mode=::NewEnumerator1
    UsePositionOffset=true
    Write Material Random=true
   )
   /SetVariables(Particles.InitialExecIndex:NiagaraFloat=&ReturnNormalizedExecIndex()) {isSetParameters=true}
  ]
  particleUpdateScript [
   &ParticleState(Kill Particles When Lifetime Has Elapsed=false)
   /SetVariables(Particles.Animate:NiagaraFloat=&ClampFloat(Float=&Subtract_Float(
      A=$Particles.NormalizedAge
      B=&NormalizeDistanceRange(
       Start Position:NiagaraPosition=$Engine.Emitter.SimulationPosition
       End Position:NiagaraPosition=$Particles.Position
       Distance:NiagaraFloat=7000
       Invert Normalized Range:NiagaraBool=false
      )
     ))) {isSetParameters=true, module=SetVariables_#1}
   &ScaleColor(
    Scale Alpha=&FloatFromCurve(
     FloatCurve=DataInterface() {propertyValues {Curve {defaultValue=1, keys [<at 0 0 RCIM_Cubic>, {tangentMode=RCTM_User, time=1, value=1}]}}}
     CurveIndex=$Particles.Animate
    )
    ScaleA=false
    ScaleRGB=false
   )
   &ScaleRibbonWidth(Ribbon Width Scale=&FloatFromCurve(
     FloatCurve=DataInterface() {propertyValues {Curve {
        defaultValue=1
        keys [
         {arriveTangent=6.8145, interpMode=RCIM_Cubic, leaveTangent=6.8145, tangentMode=RCTM_User, time=0, value=0}
         {arriveTangentWeight=0.1333, interpMode=RCIM_Cubic, tangentMode=RCTM_User, time=0.3, value=1}
        ]
       }}}
     CurveIndex=$Particles.Animate
    )) {enabled=false}
   &DynamicMaterialParameters(
    Index 0 Param 1=&FloatFromCurve(
     FloatCurve=DataInterface() {propertyValues {Curve {
        defaultValue=1
        keys [
         {arriveTangent=2.7272, interpMode=RCIM_Cubic, leaveTangent=2.7272, tangentMode=RCTM_User, time=0, value=0}
         {tangentMode=RCTM_User, time=1, value=1}
        ]
       }}}
     CurveIndex=$Particles.Animate
     Scale Curve=&FloatFromCurve(
      FloatCurve=DataInterface() {propertyValues {Curve {
         defaultValue=1
         keys [
          {arriveTangent=13.974, interpMode=RCIM_Cubic, leaveTangent=13.974, tangentMode=RCTM_User, time=0, value=0}
          {arriveTangent=2, interpMode=RCIM_Cubic, leaveTangent=2, time=0.228, value=1}
          {interpMode=RCIM_Cubic, tangentMode=RCTM_User, time=0.5, value=1}
          {arriveTangent=-4.0932, interpMode=RCIM_Cubic, leaveTangent=-4.0932, tangentMode=RCTM_User, time=1, value=0}
         ]
        }}}
      CurveIndex=$Particles.InitialExecIndex
     )
    )
    Index 0 Param 3=&Add_Float(
     A=&FloatFromCurve(
      FloatCurve=DataInterface() {propertyValues {Curve {
         defaultValue=1
         keys [
          {arriveTangent=0.5214, interpMode=RCIM_Cubic, leaveTangent=0.5214, tangentMode=RCTM_User, time=0, value=0}
          {interpMode=RCIM_Cubic, tangentMode=RCTM_User, time=3.5, value=1}
         ]
        }}}
      CurveIndex=$Particles.Age
      Scale Curve=6
     )
     B=&FloatFromCurve(
      FloatCurve=DataInterface() {propertyValues {Curve {
         defaultValue=1
         keys [{arriveTangent=2.8195, interpMode=RCIM_Cubic, leaveTangent=2.8195, tangentMode=RCTM_User, time=0, value=0}, <at 1 1 RCIM_Cubic>]
        }}}
      CurveIndex=$Particles.Animate
      Scale Curve=5
     )
    )
    Index 0 Param 4=&Waveform(
     Global Input Position=&ReturnNormalizedExecIndex(Normalized Index Scale=-1)
     Global Phase=&Multiply_Float(A=$Particles.Age, B=0.5)
     "Waveform [1]"=::NewEnumerator0
     "[1] Phase Scale"=0.2
     "[1] Frequency"=6
     "[1] Amplitude Min/Max"=<0.1 2>
    )
    Module.Index0Float0Write=true
    Module.Index0Float1Write=true
    Module.Index0Float2Write=true
    Module.Index0Float3Write=true
   )
  ]
  rendererClasses [&NiagaraSpriteRendererProperties, &NiagaraRibbonRendererProperties]
  renderers [
   {rendererClass=&NiagaraSpriteRendererProperties}
   {rendererClass=&NiagaraRibbonRendererProperties, rendererIndex=1}
  ]
 }
 FollowerRibbons {
  emitterUpdateScript [
   &EmitterState(
    Inactive Response=::NewEnumerator1
    Life Cycle Mode=::NewEnumerator1
    Loop Behavior=::NewEnumerator1
    Loop Duration Mode=::NewEnumerator1
    UseLoopDelay=false
   )
   &SpawnParticlesFromOtherEmitter(
    Attribute Reader=DataInterface() {propertyValues {EmitterBinding {bindingMode=Other, emitterName=Leaders}}}
    Spawn Rate=15
    Spawn Rate Per Particle Cap=15
    Use Maximum Age=false
    Use Minimum Age=false
   )
  ]
  particleSpawnScript [
   &InitializeParticle(
    Lifetime=2
    Ribbon Facing Mode=::NewEnumerator1
    Ribbon Width=&FloatFromCurve(FloatCurve=DataInterface() {propertyValues {Curve {keys [<at 0 1 RCIM_Cubic>,<at 5 0.3>]}}}, CurveIndex=$Emitter.Age, Scale Curve=15000)
    Ribbon Width Mode=::NewEnumerator1
    Sprite Size Mode=::NewEnumerator3
    Uniform Sprite Size=500
    UsePositionOffset=false
    Write Material Random=true
   )
   &SampleParticlesFromOtherEmitter(
    Applied Position Offset:Vector3f=<-3000 0 0>
    Mesh Scale Sampling:ENiagara_DirectReadSamplingMode=::NewEnumerator1
    Particle ID Sampling=::NewEnumerator1
    Ribbon UV Distance:ENiagara_DirectReadSamplingMode=::NewEnumerator0
   )
   /SetVariables(Particles.LeaderNormalizedAge:NiagaraFloat=&MakeFloatFromVector(Vector=$Output.SampleParticlesFromOtherEmitter.SampledScale, Channel=::NewEnumerator0)) {isSetParameters=true}
   /SetVariables(Particles.Color=&MultiplyLinearColor(
     Linear Color=&LerpLinearColors(
      StartColor=<rgba 0.4675 0.5055 1 0>
      EndColor=<rgba 1 1 1 1>
      LerpFactor=&FloatFromCurve(
       FloatCurve=DataInterface() {propertyValues {Curve {
          defaultValue=1
          keys [
           {arriveTangent=6.3376, interpMode=RCIM_Cubic, leaveTangent=6.3376, tangentMode=RCTM_User, time=0, value=0}
           {arriveTangentWeight=0.1, interpMode=RCIM_Cubic, leaveTangentWeight=0.1333, tangentMode=RCTM_User, time=0.2, value=1}
           {arriveTangentWeight=0.2325, interpMode=RCIM_Cubic, leaveTangentWeight=0.1744, tangentMode=RCTM_User, time=0.7, value=1}
           {tangentMode=RCTM_User, time=1, value=0}
          ]
         }}}
       CurveIndex=$Particles.LeaderNormalizedAge
      )
     )
     Scale Factor=&MakeVector4_FromLinearColor(LinearColor=<rgba 0.7 0.7 0.7 1>)
    )) {isSetParameters=true}
  ]
  particleUpdateScript [
   &ParticleState(Kill Particles When Lifetime Has Elapsed=false)
   /SetVariables(Particles.Animate:NiagaraFloat=$Particles.NormalizedAge) {isSetParameters=true, module=SetVariables_#1}
   &DynamicMaterialParameters(
    Index 0 Param 1=&FloatFromCurve(
     FloatCurve=DataInterface() {propertyValues {Curve {
        defaultValue=1
        keys [
         {arriveTangent=2.7272, interpMode=RCIM_Cubic, leaveTangent=2.7272, tangentMode=RCTM_User, time=0, value=0}
         {tangentMode=RCTM_User, time=1, value=1}
        ]
       }}}
     CurveIndex=$Particles.Animate
     Scale Curve=&FloatFromCurve(
      FloatCurve=DataInterface() {propertyValues {Curve {
         defaultValue=1
         keys [
          {arriveTangent=8.9712, interpMode=RCIM_Cubic, leaveTangent=8.9712, tangentMode=RCTM_User, time=0, value=0}
          {arriveTangentWeight=0.1, interpMode=RCIM_Cubic, leaveTangentWeight=0.1333, tangentMode=RCTM_User, time=0.2, value=1}
          {arriveTangentWeight=0.2325, interpMode=RCIM_Cubic, leaveTangentWeight=0.1744, tangentMode=RCTM_User, time=0.7, value=1}
          {arriveTangent=-8.1664, tangentMode=RCTM_User, time=1, value=0}
         ]
        }}}
      CurveIndex=$Particles.LeaderNormalizedAge
     )
    )
    Index 0 Param 3=&Add_Float(
     A=&FloatFromCurve(
      FloatCurve=DataInterface() {propertyValues {Curve {
         defaultValue=1
         keys [
          {arriveTangent=0.5214, interpMode=RCIM_Cubic, leaveTangent=0.5214, tangentMode=RCTM_User, time=0, value=0}
          {interpMode=RCIM_Cubic, tangentMode=RCTM_User, time=3.5, value=1}
         ]
        }}}
      CurveIndex=$Particles.Age
      Scale Curve=6
     )
     B=&FloatFromCurve(
      FloatCurve=DataInterface() {propertyValues {Curve {
         defaultValue=1
         keys [{arriveTangent=2.8195, interpMode=RCIM_Cubic, leaveTangent=2.8195, tangentMode=RCTM_User, time=0, value=0}, <at 1 1 RCIM_Cubic>]
        }}}
      CurveIndex=$Particles.Animate
      Scale Curve=4
     )
    )
    Index 0 Param 4=&Waveform(
     Global Phase=&Multiply_Float(A=$Particles.Age, B=0.5)
     "Waveform [1]"=::NewEnumerator0
     "[1] Phase Scale"=0.2
     "[1] Frequency"=6
     "[1] Amplitude Min/Max"=<0.1 2>
    )
    Module.Index0Float0Write=true
    Module.Index0Float1Write=true
    Module.Index0Float2Write=true
    Module.Index0Float3Write=true
   )
   &ScaleColor(
    Scale RGB=&MakeVectorFromLinearColorRGB(Color=&LerpLinearColors(
      StartColor=<rgba 1 1 1 1>
      EndColor=<rgba 0.2833 0.5386 0.8 0.8>
      LerpFactor=&FloatFromCurve(
       FloatCurve=DataInterface() {propertyValues {Curve {defaultValue=1, keys [<at 0 0 RCIM_Cubic>, {tangentMode=RCTM_User, time=1, value=1}]}}}
       CurveIndex=&NormalizeDistanceRange(
        Start Position:NiagaraPosition=<30000 0 0>
        End Position:NiagaraPosition=$Particles.Position
        Distance:NiagaraFloat=70000
        Invert Normalized Range:NiagaraBool=false
       )
      )
     ))
    ScaleA=true
    ScaleRGB=true
   )
  ]
  rendererClasses [&NiagaraSpriteRendererProperties, &NiagaraRibbonRendererProperties]
  renderers [
   {rendererClass=&NiagaraSpriteRendererProperties}
   {rendererClass=&NiagaraRibbonRendererProperties, rendererIndex=1}
  ]
 }
 Leaders {
  emitterUpdateScript [
   &EmitterState(Life Cycle Mode=::NewEnumerator1, Loop Behavior=::NewEnumerator1, Loop Duration=3, UseLoopDelay=false)
   &SpawnBurst_Instantaneous(Spawn Group=1, Use Spawn Probability=false, UseLoopCountLimit=false)
   &SpawnBurst_Instantaneous(Spawn Group=1, Spawn Time=0.5, Use Spawn Probability=false, UseLoopCountLimit=false) {module=SpawnBurst_Instantaneous001}
   &SpawnBurst_Instantaneous(Spawn Group=1, Spawn Time=0.8, Use Spawn Probability=false, UseLoopCountLimit=false) {module=SpawnBurst_Instantaneous002}
  ]
  particleSpawnScript [
   &PartitionParticles(Number Of Partitions=3)
   &InitializeParticle(
    Color Mode=::NewEnumerator0
    Lifetime=&MaskFloatBySpawnGroup(Masked Float=2, Passthrough Float=2.5)
    Mesh Scale Mode=::NewEnumerator3
    Position Offset=&SelectVectorFromArray(
     Direct Array Index=$Particles.PartitionParticles.Partition
     Vector Selection Array=DataInterface() {propertyValues {FloatData [<4000 0 0>, <0 -9000 0>, <0 9000 0>]}}
    )
    Sprite Size Mode=::NewEnumerator3
    Uniform Sprite Size=500
    UsePositionOffset=true
    Write Material Random=true
   )
   &AddVelocity(
    Module.Use Falloff Distance=false
    Origin Offset=<-9000 0 0>
    Override Seed=false
    Velocity Mode=::NewEnumerator1
    Velocity Speed=&MaskFloatBySpawnGroup(Masked Float=0, Passthrough Float=30000)
   )
  ]
  particleUpdateScript [
   &ParticleState()
   &ScaleMeshSize(Scale Factor=&MakeVector(
     X=&FloatFromCurve(FloatCurve=DataInterface() {propertyValues {Curve {defaultValue=0, keys [{tangentMode=RCTM_User, time=0, value=0}, {tangentMode=RCTM_User, time=1, value=1}]}}})
     Y=0
     Z=0
    ))
   /SetVariables(Particles.Velocity=&RotateVectorByQuaternion(
     Vector To Rotate=$Particles.Initial.Velocity
     Quaternion=&MakeQuaternion(
      Quaternion From=0
      Axis=<0 0 1>
      Angle=&Waveform(
       Global Phase=$Particles.Age
       Global Amplitude Scale=&RandomRangeFloat(Minimum=0.07, Maximum=0.1)
       "Waveform [1]"=::NewEnumerator12
       "[1] Frequency"=4
       "[1] Amplitude Min/Max"=<-1 1>
      )
      From=?
      To=?
      Angle Type=::NewEnumerator1
      Normalize Vectors=true
     )
    )) {isSetParameters=true}
   &ScaleSpriteSize(Uniform Curve Sprite Scale=DataInterface() {propertyValues {Curve {
       defaultValue=0
       keys [{interpMode=RCIM_Cubic, tangentMode=RCTM_User, time=0, value=3}, {interpMode=RCIM_Cubic, tangentMode=RCTM_User, time=0.1, value=1}]
      }}})
   &SolveForcesAndVelocity(Clamp Velocity=false, Limit Acceleration=false)
   &DoOnce(Trigger Condition:NiagaraBool=&SetBoolByFloatComparison(A=$Particles.NormalizedAge, B=0.4))
   &DoOnce(Trigger Condition:NiagaraBool=&SetBoolByFloatComparison(A=$Particles.NormalizedAge, B=0.8)) {module=DoOnce001}
   &GenerateLocationEvent(
    Event Generation Enabled:NiagaraBool=&Boolean_AndOperation(Bool1=&Boolean_OrOperation(Bool1=$StackContext.DoOnce.Execute, Bool2=$StackContext.DoOnce001.Execute), Bool2=true)
    Event Type:ENiagara_LocEventType=::NewEnumerator2
    Use Event Probability=false
    UseEventDelay=false
    "Vector to Send as Vector 2 (Velocity)":Vector3f=&RotateVectorByQuaternion(
     Vector To Rotate=$Particles.Initial.Velocity
     Quaternion=&MakeQuaternion(
      Quaternion From=0
      Axis=<0 0 1>
      Angle=&FloatFromCurve(
       FloatCurve=DataInterface() {propertyValues {Curve {keys [<at 0 1>,<at 0.49 0.5>,<at 0.51 -0.5>,<at 1 -1>]}}}
       CurveIndex=&RandomRangeFloat(Minimum=0, Maximum=1)
       Scale Curve=-50
      )
      From=?
      To=?
      Angle Type=::NewEnumerator0
      Normalize Vectors=true
     )
    )
   )
  ]
  rendererClasses [&NiagaraSpriteRendererProperties]
  renderers [{rendererClass=&NiagaraSpriteRendererProperties}]
 }
 Stars {
  emitterUpdateScript [
   &EmitterState()
   &SpawnParticlesFromOtherEmitter(
    Attribute Reader=DataInterface() {propertyValues {EmitterBinding {bindingMode=Other, emitterName=FollowerRibbons}}}
    Spawn Rate=3
    Spawn Rate Per Particle Cap=3
    Use Maximum Age=false
    Use Minimum Age=false
   )
  ]
  particleSpawnScript [
   &InitializeParticle(
    Color Maximum=<rgba 2 0.6631 1.5305 1>
    Color Minimum=<rgba 0.7042 0.0626 1 0.5>
    Color Mode=::NewEnumerator2
    Lifetime Max=3
    Lifetime Min=2
    Lifetime Mode=::NewEnumerator1
    Sprite Size Mode=::NewEnumerator1
    Uniform Sprite Size Max=400
    Uniform Sprite Size Min=200
    UsePositionOffset=false
    Write Material Random=true
   )
   &SampleParticlesFromOtherEmitter()
   &ShapeLocation(MaskSpawnGroup=false, Override Seed=false, Sphere Radius=1500, UseOwnerScale=true)
   &AddVelocity(
    Module.Use Falloff Distance=false
    Override Seed=false
    Velocity Mode=::NewEnumerator1
    Velocity Origin=$Output.SampleParticlesFromOtherEmitter.SampledPosition
    Velocity Speed=&RandomRangeFloat(Minimum=100, Maximum=200)
   )
  ]
  particleUpdateScript [
   &ParticleState()
   &ScaleSpriteSize(Uniform Curve Sprite Scale=DataInterface() {propertyValues {Curve {
       defaultValue=0
       keys [
        {arriveTangent=5.1504, interpMode=RCIM_Cubic, leaveTangent=5.1504, tangentMode=RCTM_User, time=0, value=0}
        {interpMode=RCIM_Cubic, tangentMode=RCTM_User, time=0.5, value=1}
        {arriveTangent=-4.4997, tangentMode=RCTM_User, time=1, value=0}
       ]
      }}})
   &VectorNoiseForce(Force Amount=3500, Module.Use Random Noise Time Interval=true, Random Noise Time Interval=0.6)
   &SolveForcesAndVelocity(Clamp Velocity=false, Limit Acceleration=false)
  ]
  rendererClasses [&NiagaraSpriteRendererProperties]
  renderers [{rendererClass=&NiagaraSpriteRendererProperties}]
 }
}
summary {
 systemName=NS_LimeRock_RealityTear
 emitters [
  {
   bEnabled=false
   emitterName=FixedRibbon
   rendererClasses [&NiagaraSpriteRendererProperties, &NiagaraRibbonRendererProperties]
  }
  {emitterName=Stars, rendererClasses [&NiagaraSpriteRendererProperties]}
  {
   emitterName=FollowerRibbons
   rendererClasses [&NiagaraSpriteRendererProperties, &NiagaraRibbonRendererProperties]
  }
  {emitterName=Leaders, rendererClasses [&NiagaraSpriteRendererProperties]}
 ]
 userVariables []
}
