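# Niagara system for the LimeRock reality tear drain effect, linking a tree position to a portal.
# Four staggered, arcing tether ribbons run from the tree to the portal, with radius from a parameter collection, and seed motes that read ribbon direction for velocity.
# Electric arc sprites switch between tree and portal rim by emitter age.
# Later bursts add vortex-driven shedding ribbons around the portal and a glow sprite with light.
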NiagaraSystem /LimeRock/Assets/FX/RealityTears/NS_LimeRock_RealityTear_Drain
dependencies {
 usedDataInterfaces [
  &NiagaraDataInterfaceCurve
  &NiagaraDataInterfaceArrayFloat3
  &NiagaraDataInterfaceParticleRead
  &NiagaraDataInterfaceColorCurve
  &NiagaraDataInterfaceArrayFloat
 ]
 usedDynamicInputs [
  &SimulationPosition
  &Vector2DFromFloat
  &TransformPosition
 ]
 usedModules [
  &self:SystemSpawnScript.SystemScriptSource.SystemScriptGraph.NiagaraNodeAssignment_2.SetVariables_#1
  &self:SystemSpawnScript.SystemScriptSource.SystemScriptGraph.NiagaraNodeAssignment_3.SetVariables_#2
  &SystemState
 ]
}
emitters {
 ElectricArcs {
  emitterUpdateScript [
   &EmitterState()
   &SpawnRate(
    Spawn Probability=0.4
    SpawnRate=&FloatFromCurve(
     FloatCurve=DataInterface() {propertyValues {Curve {
        defaultValue=1
        keys [
         {arriveTangent=1.7638, interpMode=RCIM_Cubic, time=0, value=0}
         {arriveTangentWeight=0.1094, interpMode=RCIM_Cubic, leaveTangentWeight=0.3287, tangentMode=RCTM_User, time=0.2129, value=1}
         {arriveTangentWeight=0.3239, interpMode=RCIM_Cubic, leaveTangentWeight=0.0746, tangentMode=RCTM_User, time=1, value=1}
         {arriveTangent=-7.2244, interpMode=RCIM_Cubic, tangentMode=RCTM_Break, time=1.1, value=0.1}
        ]
       }}}
     CurveIndex=$Emitter.Age
     Scale Curve=80
    )
    Use Spawn Probability=true
   )
   &SpawnBurst_Instantaneous(Spawn Count=4, Spawn Group=1, Spawn Time=0.3, Use Spawn Probability=false, UseLoopCountLimit=false) {module=SpawnBurst_Instantaneous005}
   &SpawnBurst_Instantaneous(Spawn Count=3, Spawn Group=1, Spawn Time=0.5, Use Spawn Probability=false, UseLoopCountLimit=false)
   &SpawnBurst_Instantaneous(Spawn Count=2, Spawn Time=0.6, Use Spawn Probability=false, UseLoopCountLimit=false) {module=SpawnBurst_Instantaneous001}
   &SpawnBurst_Instantaneous(Spawn Count=3, Spawn Time=1.25, Use Spawn Probability=false, UseLoopCountLimit=false) {module=SpawnBurst_Instantaneous002}
   &SpawnBurst_Instantaneous(Spawn Count=4, Spawn Time=1.5, Use Spawn Probability=false, UseLoopCountLimit=false) {module=SpawnBurst_Instantaneous003}
   &SpawnBurst_Instantaneous(Spawn Count=6, Spawn Time=1.9, Use Spawn Probability=false, UseLoopCountLimit=false) {module=SpawnBurst_Instantaneous004}
  ]
  particleSpawnScript [
   /SetVariables(Particles.TreeSpawn:NiagaraBool=&SetBoolByFloatComparison(A=$Emitter.Age, B=1.1, Comparison Type=::NewEnumerator6)) {isSetParameters=true}
   &InitializeParticle(
    Color=<rgba 4 0.1293 1.2944 1>
    Lifetime=&UniformRangedFloat(Minimum=0.09, Maximum=0.125)
    Mass Mode=::NewEnumerator1
    Position Offset=&MakeCustomVectorFromBool(
     Bool=$Particles.TreeSpawn
     True Vector=&AddVectorToPosition(
      Position=$System.LocalTreePosition
      Vector=&MakeVector(
       X=0
       Y=0
       Z=&FloatFromCurve(FloatCurve=DataInterface() {propertyValues {Curve {keys [<at 0 -0.1>,<at 1 1>]}}}, CurveIndex=$Emitter.Age, Scale Curve=1000)
      )
     )
     False Vector=<1 0 0>
    )
    Sprite Rotation Mode=::NewEnumerator2
    Sprite Size=&Multiply_Vector2D_ByFloat(
     Float=&FloatFromCurve(
      FloatCurve=DataInterface() {propertyValues {Curve {defaultValue=1, keys [<at 0 0.6 RCIM_Cubic>, {tangentMode=RCTM_User, time=1, value=1}]}}}
      CurveIndex=$Emitter.NormalizedLoopAge
     )
     Vector2D=&UniformRangedVector2D(Minimum=<1000 1000>, Maximum=<3000 3000>)
    )
    Sprite Size Mode=::NewEnumerator4
    UsePositionOffset=true
    Write Material Random=true
   ) {module=InitializeParticle001}
   &ShapeLocation(
    Cylinder Radius=&MakeCustomFloatFromBool(Bool=$Particles.TreeSpawn, True Float=100, False Float=$NPC.NPC_LimeRock.PortalRadius)
    MaskSpawnGroup=false
    Override Seed=false
    Rotation Mode=::NewEnumerator1
    Shape Primitive=::NewEnumerator1
    Surface Only=true
    Use Endcaps In Surface Only Mode=false
    UseOwnerScale=true
   )
   &CameraOffset(Camera Offset Amount=&RandomRangeFloat(Minimum=-50, Maximum=100))
  ]
  particleUpdateScript [
   &ParticleState()
   &DynamicMaterialParameters(
    Index 0 Param 1=$Particles.MaterialRandom
    Index 0 Param 2=&Lerp_Float(A=0.75, B=1.2, Alpha=$Particles.MaterialRandom)
    Index 0 Param 3=&FloatFromCurve(FloatCurve=DataInterface() {propertyValues {Curve {keys [<at 0.5 0 RCIM_Cubic>, {interpMode=RCIM_Cubic, leaveTangent=3.7979, time=1, value=1}]}}})
    Index 0 Param 4=1
    Index 1 Param 2=&FloatFromCurve(FloatCurve=DataInterface() {propertyValues {Curve {keys [<at 0.25 1 RCIM_Cubic>,<at 0.5 0>]}}}, CurveIndex=$System.Age, Scale Curve=0.5)
    Index 1 Param 3=&FloatFromCurve(
     FloatCurve=DataInterface() {propertyValues {Curve {keys [
         {arriveTangent=4.106, interpMode=RCIM_Cubic, leaveTangent=4.106, tangentMode=RCTM_User, time=0.05, value=0}
         {tangentMode=RCTM_User, time=0.35, value=1}
        ]}}}
     CurveIndex=$System.Age
    )
    Index 1 Param 4=100
    Module.Index0Float0Write=true
    Module.Index0Float1Write=true
    Module.Index0Float2Write=true
    Module.Index0Float3Write=true
    Module.Index1Float0Write=true
    Module.Index1Float1Write=true
    Module.Index1Float2Write=true
    Module.Index1Float3Write=true
    Write Parameter Index 1=::NewEnumerator1
   ) {module=DynamicMaterialParameters001}
   &ScaleSpriteSize(
    Uniform Curve Scale=&MakeCustomFloatFromBool(Bool=$Particles.TreeSpawn, True Float=0.25, False Float=1)
    Uniform Curve Sprite Scale=DataInterface() {propertyValues {Curve {keys [
        {arriveTangent=0.5666, interpMode=RCIM_Cubic, leaveTangent=0.5666, tangentMode=RCTM_User, time=0, value=0.75}
        {tangentMode=RCTM_User, time=1, value=1}
       ]}}}
   )
   &ScaleColor(
    Scale RGB=&MakeVectorFromLinearColorRGB(Color=&ColorFromCurve(ColorCurve=DataInterface() {propertyValues {
        AlphaCurve {keys []}
        BlueCurve {keys [<at 0.0806 4 RCIM_Cubic>,<at 0.4309 1.5>]}
        GreenCurve {keys [<at 0.0806 4 RCIM_Cubic>,<at 0.4309 0.7283>]}
        RedCurve {keys [<at 0.0806 4 RCIM_Cubic>,<at 0.4309 0.6641>]}
       }}))
    ScaleA=false
    ScaleRGB=true
   ) {module=ScaleColor001}
  ]
  rendererClasses [&NiagaraSpriteRendererProperties]
  renderers [{rendererClass=&NiagaraSpriteRendererProperties}]
 }
 GlowLight {
  emitterUpdateScript [
   &EmitterState()
   &SpawnBurst_Instantaneous(Spawn Time=1.4, Use Spawn Probability=false, UseLoopCountLimit=false)
  ]
  particleSpawnScript [
   &InitializeParticle(
    Color=<rgba 1 0.0343 0.3764 0.5>
    Lifetime=2
    Sprite Size Mode=::NewEnumerator3
    Uniform Sprite Size=8000
    UsePositionOffset=false
    Write Material Random=true
   )
   &DynamicMaterialParameters(
    Index 0 Param 1=3
    Index 0 Param 2=20
    Module.Index0Float0Write=true
    Module.Index0Float1Write=true
    Module.Index0Float2Write=false
    Module.Index0Float3Write=false
   )
   &CameraOffset(Camera Offset Amount=400)
   /SetVariables(Particles.LightOffset:NiagaraPosition=<6000 2000 0>) {isSetParameters=true}
  ]
  particleUpdateScript [
   &ParticleState()
   &ScaleColor(
    Scale Alpha=&FloatFromCurve(FloatCurve=DataInterface() {propertyValues {Curve {
        defaultValue=1
        keys [
         {interpMode=RCIM_Cubic, tangentMode=RCTM_User, time=0, value=0}
         {arriveTangent=-2.4028, interpMode=RCIM_Cubic, leaveTangent=-2.4028, tangentMode=RCTM_User, time=0.25, value=1}
         {tangentMode=RCTM_User, time=1, value=0}
        ]
       }}})
    ScaleA=true
    ScaleRGB=true
   )
   /SetVariables(Particles.LightColor=&Multiply_LinearColorByFloat(Linear Color=$Particles.Color, Float=15)) {isSetParameters=true}
  ]
  rendererClasses [&NiagaraSpriteRendererProperties, &NiagaraLightRendererProperties]
  renderers [
   {rendererClass=&NiagaraSpriteRendererProperties}
   {rendererClass=&NiagaraLightRendererProperties, rendererIndex=1}
  ]
 }
 SheddingRibbons {
  emitterUpdateScript [
   &EmitterState()
   &SpawnBurst_Instantaneous(Spawn Count=24, Spawn Time=1.5, Use Spawn Probability=false, UseLoopCountLimit=false)
   &SpawnBurst_Instantaneous(Spawn Count=24, Spawn Time=1.7, Use Spawn Probability=false, UseLoopCountLimit=false) {module=SpawnBurst_Instantaneous001}
  ]
  particleSpawnScript [
   &PartitionParticles(
    Normalized Index In Partition Includes One=true
    Number Of Partitions=6
    Partition Mode:ENiagara_PartitionMode_v2=::NewEnumerator1
    Write Partition Index to Ribbon Link Order:ENiagara_AttributeSamplingApplyOutput=::NewEnumerator0
    Write Partition to Ribbon ID=::NewEnumerator0
   )
   /SetVariables(Particles.Offsets:NiagaraFloat=&SelectFloatFromArray(
     Direct Array Index=$Particles.PartitionParticles.Partition
     Float Selection Array=DataInterface() {propertyValues {FloatData [0,0.25,0.79,0.4,0.6,1]}}
    )) {isSetParameters=true}
   &InitializeParticle(
    Color=&LerpLinearColors(
     StartColor=<rgba 4 0.2592 1.7958 1>
     EndColor=<rgba 1.1156 0.0783 2 0.6>
     LerpFactor=&FloatFromCurve(
      FloatCurve=DataInterface() {propertyValues {Curve {
         defaultValue=0
         keys [
          {arriveTangent=-5.8832, interpMode=RCIM_Cubic, leaveTangent=-5.8832, leaveTangentWeight=0.7475, tangentMode=RCTM_User, time=0, value=1}
          {arriveTangentWeight=0.1667, interpMode=RCIM_Cubic, leaveTangentWeight=0.1667, tangentMode=RCTM_User, time=0.5, value=0}
          {arriveTangent=5.296, tangentMode=RCTM_User, time=1, value=1}
         ]
        }}}
      CurveIndex=$Particles.PartitionParticles.Normalized Index In Partition
     )
    )
    Lifetime=&Lerp_Float(A=2.5, B=1.5, Alpha=$Particles.Offsets)
    Ribbon Width=&FloatFromCurve(
     FloatCurve=DataInterface() {propertyValues {Curve {
        defaultValue=1
        keys [
         {arriveTangent=6.1546, interpMode=RCIM_Cubic, leaveTangent=6.1546, tangentMode=RCTM_User, time=0, value=0.2}
         {arriveTangentWeight=0.1, interpMode=RCIM_Cubic, leaveTangentWeight=0.1333, tangentMode=RCTM_User, time=0.3, value=1}
         {arriveTangentWeight=0.2025, interpMode=RCIM_Cubic, leaveTangentWeight=0.1519, tangentMode=RCTM_User, time=0.7, value=1}
         {arriveTangent=-6.1895, tangentMode=RCTM_User, time=1, value=0.2}
        ]
       }}}
     CurveIndex=$Particles.PartitionParticles.Normalized Index In Partition
     Scale Curve=1000
    )
    Ribbon Width Mode=::NewEnumerator1
    UsePositionOffset=false
    Write Material Random=true
   )
   &ShapeLocation(
    Cylinder Distribution=::NewEnumerator2
    Cylinder Height=1
    Cylinder Radius=$NPC.NPC_LimeRock.PortalRadius
    MaskSpawnGroup=false
    Rotation Mode=::NewEnumerator1
    Shape Primitive=::NewEnumerator1
    U Position=&ReturnNormalizedExecIndex(Normalized Index Scale=1.25)
    UseOwnerScale=true
   )
   &AddVelocity(
    Module.Use Falloff Distance=false
    Origin Offset=<-100 0 0>
    Override Seed=false
    Velocity Mode=::NewEnumerator1
    Velocity Speed=&Lerp_Float(A=200, B=2500, Alpha=$Particles.Offsets)
   )
   &VortexForce(
    Origin Pull Only On Vortex Axis=false
    UseOriginOffset=false
    Vortex Axis=<1 0 0>
    Vortex Axis Coordinate Space:ENiagaraCoordinateSpace=::Local
    Vortex Force Amount=&Lerp_Float(A=800, B=300, Alpha=$Particles.Offsets)
   ) {module=VortexForce001}
   &ApplyInitialForces(Apply Rotational Force to Rotational Velocity=false)
  ]
  particleUpdateScript [
   &ParticleState()
   &ScaleColor(
    Scale Alpha=&FloatFromCurve(
     FloatCurve=DataInterface() {propertyValues {Curve {
        defaultValue=1
        keys [
         {arriveTangent=20.5132, interpMode=RCIM_Cubic, time=0, value=0}
         {arriveTangentWeight=0.0745, interpMode=RCIM_Cubic, leaveTangentWeight=0.2981, tangentMode=RCTM_User, time=0.15, value=1}
         <at 1 0 RCIM_Cubic>
        ]
       }}}
     Scale Curve=0.25
    )
    Scale RGB=&VectorFromFloat(Value=&FloatFromCurve(FloatCurve=DataInterface() {propertyValues {Curve {keys [
          {arriveTangent=-6.2859, interpMode=RCIM_Cubic, leaveTangent=-6.2859, tangentMode=RCTM_User, time=0, value=4}
          {arriveTangent=-2.2002, interpMode=RCIM_Cubic, leaveTangent=-2.2002, tangentMode=RCTM_User, time=0.52, value=1}
          <at 1 0.5>
         ]}}}))
    ScaleA=true
    ScaleRGB=true
   )
   &DynamicMaterialParameters(
    Index 0 Param 1=&Lerp_Float(
     A=0.8
     B=0
     Alpha=&FloatFromCurve(
      FloatCurve=DataInterface() {propertyValues {Curve {
         defaultValue=1
         keys [
          {arriveTangent=6.018, interpMode=RCIM_Cubic, leaveTangent=6.018, leaveTangentWeight=0.7925, tangentMode=RCTM_User, time=0, value=0}
          {arriveTangentWeight=0.1667, interpMode=RCIM_Cubic, leaveTangentWeight=0.1667, tangentMode=RCTM_User, time=0.5, value=1}
          {arriveTangent=-6.0708, tangentMode=RCTM_User, time=1, value=0}
         ]
        }}}
      CurveIndex=$Particles.PartitionParticles.Normalized Index In Partition
      Scale Curve=&FloatFromCurve(FloatCurve=DataInterface() {propertyValues {Curve {keys [<at 0.25 1 RCIM_Cubic>, {interpMode=RCIM_Cubic, leaveTangent=-1.4309, time=1, value=0}]}}})
     )
    )
    Index 0 Param 2=4
    Index 0 Param 3=&Add_Float(
     A=$Particles.Offsets
     B=&FloatFromCurve(
      FloatCurve=DataInterface() {propertyValues {Curve {
         defaultValue=1
         keys [
          {arriveTangent=5.5994, interpMode=RCIM_Cubic, leaveTangent=5.5994, tangentMode=RCTM_User, time=0, value=0}
          {tangentMode=RCTM_User, time=0.5, value=1}
         ]
        }}}
      Scale Curve=3
     )
    )
    Module.Index0Float0Write=true
    Module.Index0Float1Write=true
    Module.Index0Float2Write=true
    Module.Index0Float3Write=false
   )
   &ScaleRibbonWidth(Ribbon Width Scale=&FloatFromCurve(FloatCurve=DataInterface() {propertyValues {Curve {keys [{interpMode=RCIM_Cubic, leaveTangentWeight=0.2, tangentMode=RCTM_User, time=0.3, value=1}, <at 1 0 RCIM_Cubic>]}}}))
   &VortexForce(
    Origin Pull Only On Vortex Axis=false
    UseOriginOffset=false
    Vortex Axis=<1 0 0>
    Vortex Axis Coordinate Space:ENiagaraCoordinateSpace=::Local
    Vortex Force Amount=&Lerp_Float(A=3000, B=1000, Alpha=$Particles.Offsets)
   )
   &Drag(
    Drag=&FloatFromCurve(
     FloatCurve=DataInterface() {propertyValues {Curve {
        defaultValue=1
        keys [
         {arriveTangent=5.2864, interpMode=RCIM_Cubic, leaveTangent=5.2864, tangentMode=RCTM_User, time=0, value=0}
         {tangentMode=RCTM_User, time=0.5, value=1}
        ]
       }}}
     Scale Curve=&RandomRangeFloat(Minimum=1.5, Maximum=2)
    )
    Use Linear Drag=true
    Use Rotational Drag=false
   )
   &SolveForcesAndVelocity(Clamp Velocity=false, Limit Acceleration=false)
  ]
  rendererClasses [&NiagaraSpriteRendererProperties, &NiagaraRibbonRendererProperties]
  renderers [
   {rendererClass=&NiagaraSpriteRendererProperties}
   {rendererClass=&NiagaraRibbonRendererProperties, rendererIndex=1}
  ]
 }
 TetherMotes {
  emitterUpdateScript [
   &EmitterState()
   &SpawnParticlesFromOtherEmitter(
    Attribute Reader=DataInterface() {propertyValues {EmitterBinding {bindingMode=Other, emitterName=TetherRibbons}}}
    Spawn Rate=&FloatFromCurve(
     FloatCurve=DataInterface() {propertyValues {Curve {keys [<at 0 1 RCIM_Cubic>,<at 1 0>]}}}
     CurveIndex=$Emitter.NormalizedLoopAge
     Scale Curve=25
    )
    Spawn Rate Per Particle Cap=25
    Use Maximum Age=false
    Use Minimum Age=false
   )
  ]
  particleSpawnScript [
   &InitializeParticle(
    Color Maximum=<rgba 2.5 0.8289 1.9131 1>
    Color Minimum=<rgba 1.5 0.0156 0.7344 1>
    Color Mode=::NewEnumerator2
    Lifetime Max=1
    Lifetime Min=0.5
    Lifetime Mode=::NewEnumerator1
    Sprite Size Mode=::NewEnumerator1
    Uniform Sprite Size Max=22
    Uniform Sprite Size Min=8
    UsePositionOffset=false
    Write Material Random=true
   )
   &SampleParticlesFromOtherEmitter()
   &AddVelocity(
    Velocity=&self:GetRibbonDirection(
     Particle Attribute Reader:NiagaraDataInterfaceParticleRead=DataInterface() {propertyValues {EmitterBinding {bindingMode=Other, emitterName=TetherRibbons}}}
     Index:NiagaraInt32=&MakeIntFromFloat(Float=$Particles.SampleParticlesFromOtherEmitter.SampledIndex)
     IndexOffset:NiagaraInt32=1
     FallBackVelocity:Vector3f=&Multiply_VectorByFloat(
      Vector=&TransformVector(
       Vector=&NormalizeVector(
        Vector To Normalize=&Subtract_Position(A=$User.TreePosition, B=&AddVectorToPosition(Position=$Engine.Owner.Position, Vector=<0 0 4500>))
        Safe Normalize=true
        Safe Fallback Threshold=0
        Safe Fallback Vector=<0 0 0.001>
       )
       Source Space=::World
       Destination Space=::Simulation
      )
      Float=-500
     )
    )
    Velocity Speed Scale=&RandomRangeFloat(Minimum=2, Maximum=4)
   )
   &ShapeLocation(MaskSpawnGroup=false, Override Seed=false, Sphere Radius=150, Sphere Surface Distribution=1, UseOwnerScale=true)
  ]
  particleUpdateScript [
   &ParticleState()
   &ScaleSpriteSize(
    Uniform Curve Scale=&FloatFromCurve(FloatCurve=DataInterface() {propertyValues {Curve {keys [<at 6000 0.5>,<at 9000 1>,<at 15000 3>]}}}, CurveIndex=$Engine.Owner.LODDistance)
    Uniform Curve Sprite Scale=DataInterface() {propertyValues {Curve {
       defaultValue=0
       keys [
        {arriveTangent=5.1504, interpMode=RCIM_Cubic, leaveTangent=5.1504, tangentMode=RCTM_User, time=0, value=0}
        {interpMode=RCIM_Cubic, tangentMode=RCTM_User, time=0.5, value=1}
        {arriveTangent=-4.4997, tangentMode=RCTM_User, time=1, value=0}
       ]
      }}}
   )
   &VectorNoiseForce(Force Amount=8000, Module.Use Random Noise Time Interval=true, Random Noise Time Interval=0.4)
   &Drag(Drag=2, Use Linear Drag=true, Use Rotational Drag=false)
   &SolveForcesAndVelocity(Clamp Velocity=false, Limit Acceleration=false)
   &ScaleSpriteSizeBySpeed(
    Max Scale Factor=<0.7 5>
    Sample Scale Factor By Curve=true
    Scale Factor Curve=DataInterface() {propertyValues {Curve {keys [<at 0 0 RCIM_Cubic>,<at 1 1>]}}}
    Velocity Threshold=2500
   )
  ]
  rendererClasses [&NiagaraSpriteRendererProperties]
  renderers [{rendererClass=&NiagaraSpriteRendererProperties}]
 }
 TetherRibbons {
  emitterUpdateScript [
   &EmitterState()
   &SpawnBurst_Instantaneous(Spawn Count=20, Use Spawn Probability=false, UseLoopCountLimit=false)
   &SpawnBurst_Instantaneous(Spawn Count=20, Spawn Time=0.15, Use Spawn Probability=false, UseLoopCountLimit=false) {module=SpawnBurst_Instantaneous001}
   &SpawnBurst_Instantaneous(Spawn Count=20, Spawn Time=0.3, Use Spawn Probability=false, UseLoopCountLimit=false) {module=SpawnBurst_Instantaneous003}
   &SpawnBurst_Instantaneous(Spawn Count=20, Spawn Time=0.5, Use Spawn Probability=false, UseLoopCountLimit=false) {module=SpawnBurst_Instantaneous002}
  ]
  particleSpawnScript [
   &PartitionParticles(
    Normalized Index In Partition Includes One=true
    Particles Per Partition=20
    Partition Count Mode:ENiagara_InfiniteLoopDuration=::NewEnumerator1
    Partition Mode:ENiagara_PartitionMode_v2=::NewEnumerator1
    Write Partition Index to Ribbon Link Order:ENiagara_AttributeSamplingApplyOutput=::NewEnumerator0
    Write Partition to Ribbon ID=::NewEnumerator0
   )
   &InitializeParticle(
    Color=&LerpLinearColors(
     StartColor=<rgba 0.6262 0.0226 1 0.5>
     EndColor=<rgba 2 0.2799 0.9475 1>
     LerpFactor=&FloatFromCurve(
      FloatCurve=DataInterface() {propertyValues {Curve {
         defaultValue=1
         keys [
          {arriveTangent=7.1971, interpMode=RCIM_Cubic, time=0, value=0}
          {arriveTangentWeight=0.0833, interpMode=RCIM_Cubic, leaveTangentWeight=0.15, tangentMode=RCTM_User, time=0.3, value=1}
         ]
        }}}
      CurveIndex=$Particles.PartitionParticles.Normalized Index In Partition
     )
    )
    Lifetime=&Add_Float(A=1.9, B=&Multiply_FloatByInt(Float=-0.1, Integer=$Particles.PartitionParticles.Partition))
    Ribbon Facing Mode=::NewEnumerator1
    Ribbon Width=&FloatFromCurve(
     FloatCurve=DataInterface() {propertyValues {Curve {keys [
         {arriveTangent=5.3098, interpMode=RCIM_Cubic, time=0, value=1.5}
         {arriveTangentWeight=0.1491, interpMode=RCIM_Cubic, leaveTangentWeight=0.2236, tangentMode=RCTM_User, time=0.4, value=1}
         {arriveTangent=-0.5, interpMode=RCIM_Cubic, leaveTangent=-0.5, time=0.7047, value=1.1857}
         {arriveTangent=-2.2218, interpMode=RCIM_Cubic, leaveTangent=-2.2218, tangentMode=RCTM_User, time=1, value=0.7}
        ]}}}
     CurveIndex=$Particles.PartitionParticles.Normalized Index In Partition
     Scale Curve=400
    )
    Ribbon Width Mode=::NewEnumerator1
    UsePositionOffset=false
    Write Material Random=true
   )
   /SetVariables(Particles.StartPosition:NiagaraPosition=&AddVectorToPosition(
     Position=$System.LocalTreePosition
     Vector=&SelectVectorFromArray(
      Direct Array Index=$Particles.PartitionParticles.Partition
      Vector Selection Array=DataInterface() {propertyValues {FloatData [<0 0 200>, <0 0 500>, <200 0 800>, <200 0 500>]}}
     )
    )) {isSetParameters=true}
  ]
  particleUpdateScript [
   &ParticleState()
   /SetVariables(Particles.Position=&AddVectorToPosition(
     Position=&LerpPosition(
      A=$Particles.StartPosition
      B=&RotateVectorByQuaternion(
       Vector To Rotate=&Multiply_VectorByFloat(Vector=<0 0 -1>, Float=$NPC.NPC_LimeRock.PortalRadius)
       Quaternion=&MakeQuaternion(
        Quaternion From=0
        Axis=<1 0 0>
        Angle=&Add_Float(
         A=&Multiply_FloatByInt(Float=-0.3, Integer=$Particles.PartitionParticles.Partition)
         B=&FloatFromCurve(
          FloatCurve=DataInterface() {propertyValues {Curve {
             defaultValue=1
             keys [
              {arriveTangent=0.0387, interpMode=RCIM_Cubic, leaveTangent=0.0387, tangentMode=RCTM_User, time=0.5, value=0}
              {arriveTangent=4.3238, arriveTangentWeight=0.1667, interpMode=RCIM_Cubic, leaveTangent=4.3238, tangentMode=RCTM_User, time=1, value=1}
             ]
            }}}
          CurveIndex=$Particles.PartitionParticles.Normalized Index In Partition
          Scale Curve=0.7
         )
        )
        From=?
        To=?
        Angle Type=::NewEnumerator1
        Normalize Vectors=false
       )
      )
      Alpha=$Particles.PartitionParticles.Normalized Index In Partition
     )
     Vector=&Multiply_VectorByFloat(
      Vector=&SelectVectorFromArray(
       Direct Array Index=$Particles.PartitionParticles.Partition
       Vector Selection Array=DataInterface() {propertyValues {FloatData [<0 -1 0.5>, <0 1.25 0.5>, <0.5 0.4 1>, <0.5 -0.5 1>]}}
      )
      Float=&FloatFromCurve(
       FloatCurve=DataInterface() {propertyValues {Curve {keys [
           {arriveTangent=17.6728, interpMode=RCIM_Cubic, leaveTangent=17.6728, tangentMode=RCTM_User, time=0, value=0}
           <at 0.1794 1 RCIM_Cubic>
           <at 0.7 0 RCIM_Cubic>
          ]}}}
       CurveIndex=$Particles.PartitionParticles.Normalized Index In Partition
       Scale Curve=1500
      )
     )
    )) {isSetParameters=true}
   &DynamicMaterialParameters(
    Index 0 Param 1=&FloatFromCurve(
     FloatCurve=DataInterface() {propertyValues {Curve {
        defaultValue=1
        keys [
         {arriveTangent=-42.253, interpMode=RCIM_Cubic, leaveTangent=-42.253, tangentMode=RCTM_User, time=0, value=1}
         {arriveTangentWeight=0.0236, interpMode=RCIM_Cubic, leaveTangentWeight=0.2909, tangentMode=RCTM_User, time=0.06, value=0}
        ]
       }}}
     CurveIndex=$Particles.PartitionParticles.Normalized Index In Partition
    )
    Index 0 Param 2=&Add_Float(A=8, B=&Multiply_FloatByInt(Float=-1, Integer=$Particles.PartitionParticles.Partition))
    Index 0 Param 3=&FloatFromCurve(
     FloatCurve=DataInterface() {propertyValues {Curve {
        defaultValue=1
        keys [
         {arriveTangent=0.3225, interpMode=RCIM_Cubic, leaveTangent=0.3225, tangentMode=RCTM_User, time=0, value=0}
         {arriveTangent=1.7622, interpMode=RCIM_Cubic, leaveTangent=1.7622, tangentMode=RCTM_User, time=1, value=1}
        ]
       }}}
     Scale Curve=&Add_Float(A=10, B=&Multiply_FloatByInt(Float=2, Integer=$Particles.PartitionParticles.Partition))
    )
    Index 1 Param 3=&FloatFromCurve(FloatCurve=DataInterface() {propertyValues {Curve {
        defaultValue=0
        keys [
         {arriveTangent=0.7407, interpMode=RCIM_Cubic, leaveTangent=0.7407, tangentMode=RCTM_User, time=0, value=0}
         {arriveTangent=1.4461, tangentMode=RCTM_User, time=1, value=1}
        ]
       }}})
    Module.Index0Float0Write=true
    Module.Index0Float1Write=true
    Module.Index0Float2Write=true
    Module.Index0Float3Write=false
    Module.Index1Float0Write=false
    Module.Index1Float1Write=false
    Module.Index1Float2Write=true
    Module.Index1Float3Write=true
    Write Parameter Index 1=::NewEnumerator1
   )
   &ScaleRibbonWidth(Ribbon Width Scale=&FloatFromCurve(FloatCurve=DataInterface() {propertyValues {Curve {keys [
         {arriveTangent=3.2446, interpMode=RCIM_Cubic, leaveTangent=3.2446, tangentMode=RCTM_User, time=0, value=0}
         {arriveTangent=-1.8445, interpMode=RCIM_Cubic, leaveTangent=-1.8445, tangentMode=RCTM_User, time=0.3249, value=0.8128}
         {arriveTangent=2.3297, tangentMode=RCTM_User, time=1, value=1}
        ]}}}))
  ]
  rendererClasses [&NiagaraSpriteRendererProperties, &NiagaraRibbonRendererProperties]
  renderers [
   {rendererClass=&NiagaraSpriteRendererProperties}
   {rendererClass=&NiagaraRibbonRendererProperties, rendererIndex=1}
  ]
 }
}
summary {
 systemName=NS_LimeRock_RealityTear_Drain
 emitters [
  {
   emitterName=TetherRibbons
   rendererClasses [&NiagaraSpriteRendererProperties, &NiagaraRibbonRendererProperties]
  }
  {emitterName=TetherMotes, rendererClasses [&NiagaraSpriteRendererProperties]}
  {emitterName=ElectricArcs, rendererClasses [&NiagaraSpriteRendererProperties]}
  {
   emitterName=SheddingRibbons
   rendererClasses [&NiagaraSpriteRendererProperties, &NiagaraRibbonRendererProperties]
  }
  {emitterName=GlowLight, rendererClasses [&NiagaraSpriteRendererProperties, &NiagaraLightRendererProperties]}
 ]
 userVariables [
  User.EndPoint {type=NiagaraPosition, description="", defaultValue {}}
  User.TreePosition {type=NiagaraPosition, description="", defaultValue {}}
 ]
}
