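# Niagara system for the LimeRock reality tear portal: layered portal meshes, center flare with light-style component, branching rift ribbons and tethers reaching toward a tree position.
# User RiftOpen, Intensity and RiftJitter, plus an NPC-driven portal radius and system pulse, drive shape, reveal, brightness and shedding.
# Rifts and tethers are partitioned ribbons positioned procedurally each update.
# Motes read those ribbons' particles to flow along them; InteriorSmoke is disabled.
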NiagaraSystem /LimeRock/Assets/FX/RealityTears/NS_LimeRock_RealityTear_Portal
dependencies {
 usedDataInterfaces [
  &NiagaraDataInterfaceCurve
  &NiagaraDataInterfaceParticleRead
  &NiagaraDataInterfaceArrayColor
  &NiagaraDataInterfaceArrayFloat3
  &NiagaraDataInterfaceVectorCurve
  &NiagaraDataInterfaceArrayFloat
 ]
 usedDynamicInputs [
  &TransformPosition
  &NormalizeVector
  &Subtract_Position
  &ModuloFloat
 ]
 usedModules [
  &self:SystemSpawnScript.SystemScriptSource.SystemScriptGraph.NiagaraNodeAssignment_2.SetVariables_#1
  &self:SystemSpawnScript.SystemScriptSource.SystemScriptGraph.NiagaraNodeAssignment_1.SetVariables_#2
  &SystemState
  &self:SystemSpawnScript.SystemScriptSource.SystemScriptGraph.NiagaraNodeAssignment_3.SetVariables_#3
  &self:SystemSpawnScript.SystemScriptSource.SystemScriptGraph.NiagaraNodeAssignment_0.SetVariables_#4
  &self:SystemSpawnScript.SystemScriptSource.SystemScriptGraph.NiagaraNodeAssignment_5.SetVariables_#5
 ]
}
emitters {
 CenterFlare {
  emitterUpdateScript [
   &EmitterState(
    Inactive Response=::NewEnumerator1
    Life Cycle Mode=::NewEnumerator1
    Loop Behavior=::NewEnumerator1
    Loop Duration Mode=::NewEnumerator1
    UseLoopDelay=false
   )
   &SpawnBurst_Instantaneous(Use Spawn Probability=false, UseLoopCountLimit=false)
  ]
  particleSpawnScript [&InitializeParticle(
    Color=<rgba 2 0 0.12 1>
    Lifetime=0.5
    Mass=-500
    Mass Mode=::NewEnumerator1
    Sprite Size Mode=::NewEnumerator3
    Uniform Sprite Size=2500
    UsePositionOffset=false
    Write Material Random=true
   )]
  particleUpdateScript [
   &ParticleState(Kill Particles When Lifetime Has Elapsed=false)
   &ScaleSpriteSize(
    Scale Sprite Size Mode=::NewEnumerator0
    Uniform Scale Factor=&FloatFromCurve(
     FloatCurve=DataInterface() {propertyValues {Curve {keys [<at 0.7 1 RCIM_Cubic>, {arriveTangent=-5.1827, tangentMode=RCTM_User, time=1, value=0}]}}}
     CurveIndex=$System.PortalRadiusNormalized
     Scale Curve=&FloatFromCurve(
      FloatCurve=DataInterface() {propertyValues {Curve {keys [{arriveTangent=10.9428, interpMode=RCIM_Cubic, leaveTangent=10.9428, tangentMode=RCTM_User, time=0, value=0}, <at 0.2 1>]}}}
      CurveIndex=/Niagara/DynamicInputs/Math/Min_Float(A:NiagaraFloat=$User.RiftOpen, B:NiagaraFloat=&Lerp_Float(A=0.05, B=1, Alpha=$User.Intensity))
     )
    )
   )
   &DynamicMaterialParameters(
    Index 0 Param 1=0.7
    Index 0 Param 2=1
    Index 3 Param 1=2
    Index 3 Param 2=5
    Module.Index0Float0Write=true
    Module.Index0Float1Write=true
    Module.Index0Float2Write=false
    Module.Index0Float3Write=false
    Module.Index3Float0Write=true
    Module.Index3Float1Write=true
    Module.Index3Float2Write=false
    Module.Index3Float3Write=false
    Write Parameter Index 3=::NewEnumerator1
   )
   &ScaleColor(Scale Alpha=&Lerp_Float(A=1, B=0, Alpha=$System.PortalRadiusNormalized), ScaleA=true, ScaleRGB=false)
   /SetVariables(
    Particles.SpotlightIntensity:NiagaraFloat=&Lerp_Float(A=80, B=200, Alpha=$User.Intensity)
    Particles.SpotlightPosition:NiagaraPosition=&AddVectorToPosition(
     Position=&MakeVector(
      X=&Multiply_Float(A=$NPC.NPC_LimeRock.PortalRadius, B=0.5)
      Y=0
      Z=&Add_Float(A=&Multiply_Float(A=$NPC.NPC_LimeRock.PortalRadius, B=-1.1), B=-200)
     )
     Vector=<0 0 0>
    )
    Particles.SpotlightRadius:NiagaraFloat=&FloatFromCurve(
     FloatCurve=DataInterface() {propertyValues {Curve {keys [
         {arriveTangent=0.0001, interpMode=RCIM_Cubic, leaveTangent=0.0001, tangentMode=RCTM_User, time=0, value=0.5}
         {arriveTangent=0.0005, tangentMode=RCTM_User, time=1500, value=1}
        ]}}}
     CurveIndex=$NPC.NPC_LimeRock.PortalRadius
     Scale Curve=4000
    )
   ) {isSetParameters=true}
   /SetVariables(
    Particles.Darken Color:LinearColor=&Multiply_LinearColorByFloat(Linear Color=<rgba 0.0042 0.0049 0.01 0.8>, Float=$User.RiftOpen)
    Particles.Darken Size:Vector2f=&Vector2DFromFloat(Value=&FloatFromCurve(
      FloatCurve=DataInterface() {propertyValues {Curve {keys [<at 0 1400 RCIM_Cubic>,<at 200 3000>,<at 1000 4000>]}}}
      CurveIndex=$System.PortalRadius
     ))
   ) {isSetParameters=true}
   /SetVariables(
    Particles.Glow Camera Offset=300
    Particles.Glow Color=<rgba 0.7477 0.0517 1 0.15>
    Particles.Glow Size=&Multiply_Vector2D_ByFloat(Float=1, Vector2D=$Particles.SpriteSize)
   ) {isSetParameters=true}
  ]
  rendererClasses [&NiagaraSpriteRendererProperties, &NiagaraComponentRendererProperties]
  renderers [
   {rendererClass=&NiagaraSpriteRendererProperties}
   {rendererClass=&NiagaraSpriteRendererProperties, rendererIndex=1}
   {rendererClass=&NiagaraSpriteRendererProperties, rendererIndex=2}
   {rendererClass=&NiagaraComponentRendererProperties, rendererIndex=3}
  ]
 }
 InteriorSmoke {
  bEnabled=false
  emitterUpdateScript [
   &EmitterState()
   &SpawnRate(SpawnRate=&Multiply_Float(A=$System.PortalRadiusNormalized, B=10), Use Spawn Probability=false)
  ]
  particleSpawnScript [
   &InitializeParticle(
    Lifetime=3
    Sprite Rotation Mode=::NewEnumerator2
    Sprite Size Mode=::NewEnumerator1
    Uniform Sprite Size Max=1700
    Uniform Sprite Size Min=1400
    UsePositionOffset=false
    Write Material Random=true
   )
   &CameraOffset(Camera Offset Amount=200)
   &DynamicMaterialParameters(
    Index 2 Param 1=1
    Index 2 Param 2=0.4
    Module.Index2Float0Write=true
    Module.Index2Float1Write=true
    Module.Index2Float2Write=false
    Module.Index2Float3Write=false
    Write Parameter Index 0=::NewEnumerator0
    Write Parameter Index 2=::NewEnumerator1
   )
  ]
  particleUpdateScript [
   &ParticleState()
   &ShapeLocation(
    MaskSpawnGroup=false
    Offset=&MakeVector(
     X=&Multiply_Float(
      A=/Niagara/DynamicInputs/Math/CalculateParticleRadius(
       Radius Calculation Type:ENiagaraCollisionRadiusOptions=::NewEnumerator0
       Particle Radius:NiagaraFloat=1
       Mesh Dimensions:Vector3f=<1 1 1>
       Method for Calculating Particle Radius:ENiagaraCalculateRadiusOptions=::NewEnumerator2
      )
      B=-1.5
     )
     Y=0
     Z=&Add_Float(
      A=300
      B=&Multiply_Float(A=$Particles.Age, B=&RandomRangeFloat(Minimum=-100, Maximum=-500))
     )
    )
    Offset Mode=::NewEnumerator0
    Override Seed=false
    Sphere Radius=$System.PortalRadius
    UseOwnerScale=true
   )
   &ScaleColor(
    Scale Alpha=&FloatFromCurve(FloatCurve=DataInterface() {propertyValues {Curve {
        defaultValue=0
        keys [
         {interpMode=RCIM_Cubic, tangentMode=RCTM_User, time=0, value=0}
         {interpMode=RCIM_Cubic, tangentMode=RCTM_User, time=0.25, value=1}
         {tangentMode=RCTM_User, time=1, value=0}
        ]
       }}})
    ScaleA=true
    ScaleRGB=false
   )
  ]
  rendererClasses [&NiagaraSpriteRendererProperties]
  renderers [
   {rendererClass=&NiagaraSpriteRendererProperties}
   {rendererClass=&NiagaraSpriteRendererProperties, rendererIndex=1}
  ]
 }
 MainRift {
  emitterUpdateScript [
   &EmitterState(
    Inactive Response=::NewEnumerator1
    Life Cycle Mode=::NewEnumerator1
    Loop Behavior=::NewEnumerator1
    Loop Duration Mode=::NewEnumerator1
    UseLoopDelay=false
   )
   &SpawnBurst_Instantaneous(Spawn Count=14, Use Spawn Probability=false, UseLoopCountLimit=false)
  ]
  particleSpawnScript [
   &PartitionParticles(
    Normalized Index In Partition Includes One=true
    Number Of Partitions=1
    Particles Per Partition=14
    Partition Mode:ENiagara_PartitionMode_v2=::NewEnumerator1
    Write Partition Index to Ribbon Link Order:ENiagara_AttributeSamplingApplyOutput=::NewEnumerator0
    Write Partition to Ribbon ID=::NewEnumerator0
   )
   &InitializeParticle(
    Color=&LerpLinearColors(
     StartColor=<rgba 0.2633 0.2648 0.724 0.25>
     EndColor=<rgba 0.7 0.5261 0.6423 0.9>
     LerpFactor=&FloatFromCurve(
      FloatCurve=DataInterface() {propertyValues {Curve {
         defaultValue=1
         keys [
          {arriveTangent=26.9494, interpMode=RCIM_Cubic, leaveTangent=26.9494, tangentMode=RCTM_User, time=0, value=0}
          {arriveTangentWeight=0.1, interpMode=RCIM_Cubic, leaveTangentWeight=0.1333, tangentMode=RCTM_User, time=0.1, value=1}
          {arriveTangentWeight=0.2325, interpMode=RCIM_Cubic, leaveTangentWeight=0.1744, tangentMode=RCTM_User, time=0.5, value=1}
          {arriveTangent=-4.0283, tangentMode=RCTM_User, time=1, value=0}
         ]
        }}}
      CurveIndex=$Particles.PartitionParticles.Normalized Index In Partition
     )
    )
    Lifetime=2
    Ribbon Facing Mode=::NewEnumerator1
    Ribbon Width=&FloatFromCurve(
     FloatCurve=DataInterface() {propertyValues {Curve {keys [<at 0.5 1 RCIM_Cubic>, {arriveTangent=-2.9556, tangentMode=RCTM_User, time=1, value=0.5}]}}}
     CurveIndex=$Particles.PartitionParticles.Normalized Index In Partition
     Scale Curve=1100
    )
    Ribbon Width Mode=::NewEnumerator1
    UsePositionOffset=false
    Write Material Random=true
   )
  ]
  particleUpdateScript [
   &ParticleState(Kill Particles When Lifetime Has Elapsed=false)
   /SetVariables(Particles.Animate:NiagaraFloat=&ClampFloat(Float=&Subtract_Float(
      A=$User.RiftOpen
      B=&NormalizeDistanceRange(
       Start Position:NiagaraPosition=$Engine.Emitter.SimulationPosition
       End Position:NiagaraPosition=$Particles.Position
       Distance:NiagaraFloat=5000
       Invert Normalized Range:NiagaraBool=false
      )
     ))) {isSetParameters=true, module=SetVariables_#6}
   /SetVariables(Particles.EndPoint:Vector3f=&Lerp_Vector(
     A=<2200 500 -2000>
     B=<4800 1500 -2500>
     Alpha=&FloatFromCurve(
      FloatCurve=DataInterface() {propertyValues {Curve {keys [<at 100 0 RCIM_Cubic>,<at 1400 1 RCIM_Cubic>]}}}
      CurveIndex=$NPC.NPC_LimeRock.PortalRadius
     )
    )) {isSetParameters=true}
   /SetVariables(Particles.Position=&RotateVectorByQuaternion(
     Vector To Rotate=&Lerp_Vector(
      A=&Multiply_VectorByFloat(Vector=<0 0 -1>, Float=$System.PortalRadius)
      B=$Particles.EndPoint
      Alpha=&FloatFromCurve(
       FloatCurve=DataInterface() {propertyValues {Curve {
          defaultValue=1
          keys [
           {arriveTangent=0.0999, interpMode=RCIM_Cubic, leaveTangent=0.0999, tangentMode=RCTM_User, time=0, value=0}
           {arriveTangent=2.2106, interpMode=RCIM_Cubic, leaveTangent=2.2106, tangentMode=RCTM_User, time=1, value=1}
          ]
         }}}
       CurveIndex=$Particles.PartitionParticles.Normalized Index In Partition
      )
     )
     Quaternion=&MakeQuaternion(
      Quaternion From=0
      Axis=<1 0 0>
      Angle=&Add_Float(
       A=0
       B=&FloatFromCurve(
        FloatCurve=DataInterface() {propertyValues {Curve {
           defaultValue=1
           keys [
            {arriveTangent=-1.6366, interpMode=RCIM_Cubic, leaveTangent=-1.6366, leaveTangentWeight=0.8544, tangentMode=RCTM_User, time=0, value=1}
            {arriveTangent=0.0387, interpMode=RCIM_Cubic, leaveTangent=0.0387, tangentMode=RCTM_User, time=0.681, value=-0.1436}
            {arriveTangent=0.3384, arriveTangentWeight=0.1667, interpMode=RCIM_Cubic, leaveTangent=0.3384, tangentMode=RCTM_User, time=1, value=0}
           ]
          }}}
        CurveIndex=$Particles.PartitionParticles.Normalized Index In Partition
        Scale Curve=0.4
       )
      )
      From=?
      To=?
      Angle Type=::NewEnumerator1
      Normalize Vectors=false
     )
    )) {isSetParameters=true, module=SetVariables_#7}
   &DynamicMaterialParameters(
    Index 0 Param 1=&Min_Float(
     A:NiagaraFloat=&FloatFromCurve(
      FloatCurve=DataInterface() {propertyValues {Curve {
         defaultValue=1
         keys [
          {arriveTangent=2.7272, interpMode=RCIM_Cubic, leaveTangent=2.7272, tangentMode=RCTM_User, time=0, value=0}
          {tangentMode=RCTM_User, time=1, value=1}
         ]
        }}}
      CurveIndex=$Particles.Animate
      Scale Curve=&FloatFromCurve(
       FloatCurve=DataInterface() {propertyValues {Curve {
          defaultValue=1
          keys [
           {arriveTangent=23.1802, interpMode=RCIM_Cubic, leaveTangent=23.1802, tangentMode=RCTM_User, time=0, value=0}
           {arriveTangentWeight=0.085, interpMode=RCIM_Cubic, leaveTangentWeight=0.1014, tangentMode=RCTM_User, time=0.1, value=1}
           {arriveTangent=-1.3131, interpMode=RCIM_Cubic, leaveTangent=-1.3131, tangentMode=RCTM_User, time=0.7448, value=0.7176}
           {arriveTangent=-5.2394, interpMode=RCIM_Cubic, leaveTangent=-5.2394, tangentMode=RCTM_User, time=1, value=0}
          ]
         }}}
       CurveIndex=$Particles.PartitionParticles.Normalized Index In Partition
      )
     )
     B:NiagaraFloat=&Lerp_Float(A=0.4, B=1, Alpha=$User.Intensity)
    )
    Index 0 Param 3=&Add_Float(
     A=&Multiply_Float(A=$Particles.Age, B=1)
     B=&FloatFromCurve(
      FloatCurve=DataInterface() {propertyValues {Curve {
         defaultValue=1
         keys [{arriveTangent=2.8195, interpMode=RCIM_Cubic, leaveTangent=2.8195, tangentMode=RCTM_User, time=0, value=0}, <at 1 1 RCIM_Cubic>]
        }}}
      CurveIndex=$Particles.Animate
      Scale Curve=5
     )
    )
    Index 0 Param 4=&Waveform(
     Global Input Position=$Particles.PartitionParticles.Normalized Index In Partition
     "Waveform [1]"=::NewEnumerator0
     "[1] Phase Scale"=0.2
     "[1] Frequency"=3
     "[1] Amplitude Min/Max"=<-1 2>
    )
    Module.Index0Float0Write=true
    Module.Index0Float1Write=false
    Module.Index0Float2Write=true
    Module.Index0Float3Write=true
   )
   &ScaleColor(
    Scale RGB=&VectorFromFloat(Value=&Lerp_Float(A=1, B=2, Alpha=$System.RiftPulse))
    ScaleA=false
    ScaleRGB=true
   )
  ]
  rendererClasses [&NiagaraSpriteRendererProperties, &NiagaraRibbonRendererProperties]
  renderers [
   {rendererClass=&NiagaraSpriteRendererProperties}
   {rendererClass=&NiagaraRibbonRendererProperties, rendererIndex=1}
  ]
 }
 Portal {
  emitterUpdateScript [
   &EmitterState(
    Inactive Response=::NewEnumerator1
    Life Cycle Mode=::NewEnumerator1
    Loop Behavior=::NewEnumerator1
    Loop Duration Mode=::NewEnumerator1
    UseLoopDelay=false
   )
   &SpawnBurst_Instantaneous(Spawn Count=3, Use Spawn Probability=false, UseLoopCountLimit=false)
  ]
  particleSpawnScript [
   &PartitionParticles(Number Of Partitions=3)
   &InitializeParticle(
    Color=&SelectLinearColorFromArray(
     Array Sampling Mode=::NewEnumerator1
     Direct Array Index=$Particles.PartitionParticles.Partition
     Color Selection Array=DataInterface() {propertyValues {ColorData [<rgba 0.7293 0.0378 1 1>, <rgba 1 1 1 6>, <rgba 1 1 1 1>]}}
     Random Seed=?
    )
    Lifetime=4.5
    Mesh Scale=&SelectVectorFromArray(
     Direct Array Index=$Particles.PartitionParticles.Partition
     Vector Selection Array=DataInterface() {propertyValues {FloatData [<8 8 8>, <0.26 0.26 0.4>, <16 16 0.02>]}}
    )
    Mesh Scale Mode=::NewEnumerator4
    Position Offset=&SelectVectorFromArray(
     Direct Array Index=$Particles.PartitionParticles.Partition
     Vector Selection Array=DataInterface() {propertyValues {FloatData [<0 0 0>, <180 0 0>, <0 0 0>]}}
    )
    UsePositionOffset=true
    Write Material Random=true
   )
   &InitialMeshOrientation(Mesh Orientation Mode=::NewEnumerator2, Use Rotation Vector=false)
  ]
  particleUpdateScript [
   &ParticleState(Kill Particles When Lifetime Has Elapsed=false)
   &UpdateMeshOrientation(Rotation Rate=&FloatFromCurve(
     FloatCurve=DataInterface() {propertyValues {Curve {keys [<at 1 0 RCIM_Cubic>, {arriveTangent=2.3834, tangentMode=RCTM_User, time=2, value=1}]}}}
     CurveIndex=$User.Intensity
     Scale Curve=0.1
    ))
   &DynamicMaterialParameters(
    Index 0 Param 1=&FloatFromCurve(
     FloatCurve=DataInterface() {propertyValues {Curve {keys [{arriveTangent=-2.1015, interpMode=RCIM_Cubic, leaveTangent=-2.1015, tangentMode=RCTM_User, time=0, value=0.5}, <at 1 -0.15>]}}}
     Scale Curve=&RandomRangeFloat(Minimum=0.5, Maximum=1.5)
    )
    Index 0 Param 2=&FloatFromCurve(
     FloatCurve=DataInterface() {propertyValues {Curve {keys [{arriveTangent=-0.3464, interpMode=RCIM_Cubic, leaveTangent=-0.3464, tangentMode=RCTM_User, time=0, value=1}, <at 1 0.9>]}}}
     Scale Curve=&RandomRangeFloat(Minimum=0.5, Maximum=1)
    )
    Index 0 Param 3=&FloatFromCurve(
     FloatCurve=DataInterface() {propertyValues {Curve {keys [
         {interpMode=RCIM_Cubic, tangentMode=RCTM_Break, time=0, value=0}
         {arriveTangent=0.4478, interpMode=RCIM_Cubic, leaveTangent=0.4478, tangentMode=RCTM_User, time=0.15, value=1}
         {interpMode=RCIM_Cubic, tangentMode=RCTM_Break, time=1, value=1.2}
        ]}}}
     Scale Curve=&RandomRangeFloat(Minimum=0.5, Maximum=1)
    )
    Module.Index0Float0Write=true
    Module.Index0Float1Write=true
    Module.Index0Float2Write=true
    Module.Index0Float3Write=true
   )
   &ScaleColor(
    Scale Alpha=&FloatFromCurve(
     FloatCurve=DataInterface() {propertyValues {Curve {
        defaultValue=1
        keys [{interpMode=RCIM_Cubic, tangentMode=RCTM_User, time=0, value=0}, {interpMode=RCIM_Cubic, tangentMode=RCTM_User, time=0.2, value=1}]
       }}}
     CurveIndex=$User.RiftOpen
    )
    Scale RGB=&VectorFromFloat(Value=&Lerp_Float(A=1, B=1.5, Alpha=$System.RiftPulse))
    ScaleA=true
    ScaleRGB=true
   )
   &ScaleMeshSize(Scale Factor=&Multiply_VectorByFloat(Vector=<1 1 1>, Float=&Multiply_Float(A=0.0007, B=$System.PortalRadius)))
  ]
  rendererClasses [&NiagaraMeshRendererProperties]
  renderers [
   {rendererClass=&NiagaraMeshRendererProperties, rendererIndex=0}
   {rendererClass=&NiagaraMeshRendererProperties, rendererIndex=1}
   {rendererClass=&NiagaraMeshRendererProperties, rendererIndex=2}
  ]
 }
 RiftMotes {
  emitterUpdateScript [
   &EmitterState()
   &SpawnRate(SpawnGroup=1, SpawnRate=&Multiply_Float(A=$System.PortalRadius, B=0.1), Use Spawn Probability=false)
   &SpawnParticlesFromOtherEmitter(
    Attribute Reader=DataInterface() {propertyValues {EmitterBinding {bindingMode=Other, emitterName=MainRift}}}
    Spawn Rate=3
    Spawn Rate Per Particle Cap=3
    Use Maximum Age=false
    Use Minimum Age=false
   )
  ]
  particleSpawnScript [
   &InitializeParticle(
    Color Maximum=<rgba 2.5 0.8289 1.9131 1>
    Color Minimum=<rgba 1.5 0.0156 0.7344 1>
    Color Mode=::NewEnumerator2
    Lifetime Mode=::NewEnumerator1
    Sprite Size Mode=::NewEnumerator1
    Uniform Sprite Size Max=25
    Uniform Sprite Size Min=12
    UsePositionOffset=false
    Write Material Random=true
   )
   &SampleParticlesFromOtherEmitter()
   &AddVelocity(
    Velocity=&self:GetRibbonDirection(
     Particle Attribute Reader:NiagaraDataInterfaceParticleRead=DataInterface() {propertyValues {EmitterBinding {bindingMode=Other, emitterName=MainRift}}}
     Index:NiagaraInt32=&MakeIntFromFloat(Float=$Particles.SampleParticlesFromOtherEmitter.SampledIndex)
     IndexOffset:NiagaraInt32=-1
     FallBackVelocity:Vector3f=<0 0 0>
    )
    Velocity Speed Scale=&MaskFloatBySpawnGroup(Masked Float=&RandomRangeFloat(Minimum=1, Maximum=2), Passthrough Float=0)
   )
   &ShapeLocation(MaskSpawnGroup=false, Override Seed=false, Sphere Surface Distribution=1, UseOwnerScale=true)
   &ShapeLocation(
    Apply To Particle Position=::NewEnumerator1
    Cylinder Height=40
    Cylinder Radius=$System.PortalRadius
    MaskSpawnGroup=false
    Override Seed=false
    Rotation Mode=::NewEnumerator1
    Shape Origin=<0 0 0>
    Shape Primitive=::NewEnumerator1
    Surface Only=true
    Use Endcaps In Surface Only Mode=false
    UseOwnerScale=true
   ) {module=ShapeLocation001}
   /SetVariables(Particles.Position=&MaskVectorBySpawnGroup(
     Masked Vector=Linked() {linkedVariable {name=Particles.Position, type=NiagaraPosition}}
     Passthrough Vector=Linked() {linkedVariable {name=Particles.ShapeLocation001.ShapePosition, type=NiagaraPosition}}
    )) {isSetParameters=true}
  ]
  particleUpdateScript [
   &ParticleState()
   &ScaleSpriteSize(
    Uniform Curve Scale=&FloatFromCurve(FloatCurve=DataInterface() {propertyValues {Curve {keys [<at 6000 1>,<at 15000 3>]}}}, CurveIndex=$Engine.Owner.LODDistance)
    Uniform Curve Sprite Scale=DataInterface() {propertyValues {Curve {
       defaultValue=0
       keys [
        {arriveTangent=5.1504, interpMode=RCIM_Cubic, leaveTangent=5.1504, tangentMode=RCTM_User, time=0, value=0}
        {interpMode=RCIM_Cubic, tangentMode=RCTM_User, time=0.5, value=1}
        {arriveTangent=-4.4997, tangentMode=RCTM_User, time=1, value=0}
       ]
      }}}
   )
   &VectorNoiseForce(Force Amount=2000, Module.Use Random Noise Time Interval=true, Random Noise Time Interval=0.6)
   &VortexForce(
    Influence Falloff Radius:NiagaraFloat=&Multiply_Float(A=$System.PortalRadius, B=2)
    Use Influence Falloff=true
    UseOriginOffset=false
    Vortex Axis=<1 0 0>
    Vortex Axis Coordinate Space:ENiagaraCoordinateSpace=::Local
    Vortex Force Amount=&Add_Float(A=$System.PortalRadius, B=150)
   )
   &Drag(Use Linear Drag=true, Use Rotational Drag=false)
   &SolveForcesAndVelocity(Clamp Velocity=false, Limit Acceleration=false)
   &ScaleSpriteSizeBySpeed(
    Max Scale Factor=<0.7 5>
    Sample Scale Factor By Curve=true
    Scale Factor Curve=DataInterface() {propertyValues {Curve {keys [<at 0 0 RCIM_Cubic>,<at 1 1>]}}}
    Velocity Threshold=1500
   )
   &ScaleColor(
    Scale RGB=&VectorFromFloat(Value=&Lerp_Float(A=1, B=5, Alpha=$System.RiftPulse))
    ScaleA=false
    ScaleRGB=true
   )
  ]
  rendererClasses [&NiagaraSpriteRendererProperties]
  renderers [{rendererClass=&NiagaraSpriteRendererProperties}]
 }
 SheddingRibbons {
  emitterUpdateScript [
   &EmitterState(
    Inactive Response=::NewEnumerator1
    Life Cycle Mode=::NewEnumerator1
    Loop Duration=&Lerp_Float(
     A=&RandomRangeFloat(Minimum=0.8, Maximum=1.4)
     B=&RandomRangeFloat(Minimum=0.14, Maximum=0.25)
     Alpha=$User.RiftJitter
    )
    Recalculate Duration Each Loop=true
    UseLoopDelay=false
   )
   /SetVariables(Emitter.EmitterRandom:NiagaraFloat=&RandomRangeFloat(Minimum=0, Maximum=1, Evaluation Type=::NewEnumerator1)) {isSetParameters=true}
   &SpawnBurst_Instantaneous(
    Spawn Count=&MakeCustomIntFromBool(Bool=&SetBoolByFloatComparison(A=$System.PortalRadius, B=100), True INT=6, False INT=0)
    Use Spawn Probability=false
    UseLoopCountLimit=false
   )
  ]
  particleSpawnScript [
   /SetVariables(Particles.RibbonID=/Niagara/DynamicInputs/NiagaraID/MakeNiagaraIDFromInt(Index:NiagaraInt32=$Emitter.LoopCount, Acquire Tag:NiagaraInt32=0)) {isSetParameters=true}
   /SetVariables(Particles.InitialExecIndex:NiagaraFloat=&ReturnNormalizedExecIndex()) {isSetParameters=true}
   &InitializeParticle(
    Color=&LerpLinearColors(
     StartColor=<rgba 1 0.2135 0.5944 1>
     EndColor=<rgba 0.4198 0.1445 0.75 0.6>
     LerpFactor=&FloatFromCurve(
      FloatCurve=DataInterface() {propertyValues {Curve {
         defaultValue=0
         keys [
          {arriveTangent=-5.8832, interpMode=RCIM_Cubic, leaveTangent=-5.8832, leaveTangentWeight=0.7475, tangentMode=RCTM_User, time=0, value=1}
          {arriveTangentWeight=0.1667, interpMode=RCIM_Cubic, leaveTangentWeight=0.1667, tangentMode=RCTM_User, time=0.5, value=0}
          {arriveTangent=5.296, tangentMode=RCTM_User, time=1, value=1}
         ]
        }}}
      CurveIndex=$Particles.InitialExecIndex
     )
    )
    Lifetime=&Multiply_Float(
     A=&Lerp_Float(A=1.5, B=2.5, Alpha=$Emitter.EmitterRandom)
     B=&Lerp_Float(A=1, B=0.3, Alpha=$User.RiftJitter)
    )
    Ribbon Width=&FloatFromCurve(
     FloatCurve=DataInterface() {propertyValues {Curve {
        defaultValue=1
        keys [
         {arriveTangent=4.2038, interpMode=RCIM_Cubic, leaveTangent=4.2038, tangentMode=RCTM_User, time=0, value=0.2}
         {interpMode=RCIM_Cubic, tangentMode=RCTM_User, time=0.5, value=1}
         {arriveTangent=-4.0536, tangentMode=RCTM_User, time=1, value=0.2}
        ]
       }}}
     CurveIndex=$Particles.InitialExecIndex
     Scale Curve=300
    )
    Ribbon Width Mode=::NewEnumerator1
    UsePositionOffset=false
    Write Material Random=true
   )
   &ShapeLocation(
    Cylinder Distribution=::NewEnumerator2
    Cylinder Height=1
    Cylinder Radius=$System.PortalRadius
    MaskSpawnGroup=false
    Rotation Mode=::NewEnumerator1
    Shape Primitive=::NewEnumerator1
    U Position=&Add_Float(
     A=&ReturnNormalizedExecIndex(Normalized Index Scale=&Lerp_Float(A=0.4, B=0.9, Alpha=$Emitter.EmitterRandom))
     B=&Multiply_Float(A=$Engine.System.Age, B=1.5)
    )
    UseOwnerScale=true
   )
   &AddVelocity(
    Module.Use Falloff Distance=false
    Origin Offset=<-100 0 0>
    Override Seed=false
    Velocity Mode=::NewEnumerator1
    Velocity Speed=&Lerp_Float(A=300, B=1500, Alpha=$User.RiftJitter)
   )
   &VortexForce(
    Origin Pull Only On Vortex Axis=false
    UseOriginOffset=false
    Vortex Axis=<1 0 0>
    Vortex Axis Coordinate Space:ENiagaraCoordinateSpace=::Local
    Vortex Force Amount=&Lerp_Float(A=500, B=1000, Alpha=$User.RiftJitter)
   ) {module=VortexForce001}
   &ApplyInitialForces(Apply Rotational Force to Rotational Velocity=false)
  ]
  particleUpdateScript [
   &ParticleState()
   &ScaleColor(
    Scale Alpha=&FloatFromCurve(
     FloatCurve=DataInterface() {propertyValues {Curve {
        defaultValue=1
        keys [
         {arriveTangent=20.5132, interpMode=RCIM_Cubic, time=0, value=0}
         {arriveTangentWeight=0.0745, interpMode=RCIM_Cubic, leaveTangentWeight=0.2981, tangentMode=RCTM_User, time=0.15, value=1}
         {arriveTangentWeight=0.1333, interpMode=RCIM_Cubic, leaveTangentWeight=0.1667, tangentMode=RCTM_User, time=0.4, value=1}
         {arriveTangent=-3.5203, tangentMode=RCTM_User, time=1, value=0}
        ]
       }}}
     Scale Curve=0.3
    )
    ScaleA=true
    ScaleRGB=false
   )
   &DynamicMaterialParameters(
    Index 0 Param 1=&Lerp_Float(
     A=0.8
     B=0
     Alpha=&FloatFromCurve(
      FloatCurve=DataInterface() {propertyValues {Curve {
         defaultValue=1
         keys [
          {arriveTangent=6.018, interpMode=RCIM_Cubic, leaveTangent=6.018, leaveTangentWeight=0.7925, tangentMode=RCTM_User, time=0, value=0}
          {arriveTangentWeight=0.1667, interpMode=RCIM_Cubic, leaveTangentWeight=0.1667, tangentMode=RCTM_User, time=0.5, value=1}
          {arriveTangent=-6.0708, tangentMode=RCTM_User, time=1, value=0}
         ]
        }}}
      CurveIndex=$Particles.InitialExecIndex
      Scale Curve=&FloatFromCurve(FloatCurve=DataInterface() {propertyValues {Curve {keys [<at 0.3 1 RCIM_Cubic>, {arriveTangent=-1.4309, interpMode=RCIM_Cubic, leaveTangent=-1.4309, tangentMode=RCTM_User, time=1, value=0}]}}})
     )
    )
    Index 0 Param 2=4
    Index 0 Param 3=&Multiply_Float(A=$Emitter.Age, B=-2)
    Module.Index0Float0Write=true
    Module.Index0Float1Write=true
    Module.Index0Float2Write=true
    Module.Index0Float3Write=false
   )
   &ScaleRibbonWidth(Ribbon Width Scale=&FloatFromCurve(FloatCurve=DataInterface() {propertyValues {Curve {keys [{arriveTangent=2.5115, interpMode=RCIM_Cubic, time=0.5, value=1}, {arriveTangent=-1.744, tangentMode=RCTM_User, time=1, value=0}]}}}))
   &VortexForce(
    Origin Pull Only On Vortex Axis=false
    UseOriginOffset=false
    Vortex Axis=<1 0 0>
    Vortex Axis Coordinate Space:ENiagaraCoordinateSpace=::Local
    Vortex Force Amount=600
   )
   &Drag(
    Drag=&FloatFromCurve(
     FloatCurve=DataInterface() {propertyValues {Curve {
        defaultValue=1
        keys [
         {arriveTangent=5.2864, interpMode=RCIM_Cubic, leaveTangent=5.2864, tangentMode=RCTM_User, time=0, value=0}
         {tangentMode=RCTM_User, time=0.5, value=1}
        ]
       }}}
     Scale Curve=&RandomRangeFloat(Minimum=0.5, Maximum=2.5)
    )
    Use Linear Drag=true
    Use Rotational Drag=false
   )
   &SolveForcesAndVelocity(Clamp Velocity=false, Limit Acceleration=false)
  ]
  rendererClasses [&NiagaraSpriteRendererProperties, &NiagaraRibbonRendererProperties]
  renderers [
   {rendererClass=&NiagaraSpriteRendererProperties}
   {rendererClass=&NiagaraRibbonRendererProperties, rendererIndex=1}
  ]
 }
 SideRifts {
  emitterUpdateScript [
   &EmitterState(
    Inactive Response=::NewEnumerator1
    Life Cycle Mode=::NewEnumerator1
    Loop Behavior=::NewEnumerator1
    Loop Duration Mode=::NewEnumerator1
    UseLoopDelay=false
   )
   &SpawnBurst_Instantaneous(Spawn Count=20, Use Spawn Probability=false, UseLoopCountLimit=false)
  ]
  particleSpawnScript [
   &PartitionParticles(
    Normalized Index In Partition Includes One=true
    Particles Per Partition=10
    Partition Mode:ENiagara_PartitionMode_v2=::NewEnumerator1
    Write Partition Index to Ribbon Link Order:ENiagara_AttributeSamplingApplyOutput=::NewEnumerator0
    Write Partition to Ribbon ID=::NewEnumerator0
   )
   &InitializeParticle(
    Color=&LerpLinearColors(
     StartColor=<rgba 0.2633 0.2648 0.724 0.1>
     EndColor=<rgba 0.7 0.5261 0.6423 0.9>
     LerpFactor=&FloatFromCurve(
      FloatCurve=DataInterface() {propertyValues {Curve {
         defaultValue=1
         keys [
          {arriveTangent=26.9494, interpMode=RCIM_Cubic, leaveTangent=26.9494, tangentMode=RCTM_User, time=0, value=0}
          {arriveTangentWeight=0.1, interpMode=RCIM_Cubic, leaveTangentWeight=0.1333, tangentMode=RCTM_User, time=0.1, value=1}
          {arriveTangentWeight=0.2325, interpMode=RCIM_Cubic, leaveTangentWeight=0.1744, tangentMode=RCTM_User, time=0.5, value=1}
          {arriveTangent=-3.987, tangentMode=RCTM_User, time=1, value=0}
         ]
        }}}
      CurveIndex=$Particles.PartitionParticles.Normalized Index In Partition
      Scale Curve=0.5
     )
    )
    Lifetime=2
    Ribbon Facing Mode=::NewEnumerator1
    Ribbon Width=&FloatFromCurve(
     FloatCurve=DataInterface() {propertyValues {Curve {keys [<at 0.5 1 RCIM_Cubic>, {arriveTangent=-1.564, tangentMode=RCTM_User, time=1, value=0.5}]}}}
     CurveIndex=$Particles.PartitionParticles.Normalized Index In Partition
     Scale Curve=600
    )
    Ribbon Width Mode=::NewEnumerator1
    UsePositionOffset=false
    Write Material Random=true
   )
  ]
  particleUpdateScript [
   &ParticleState(Kill Particles When Lifetime Has Elapsed=false)
   /SetVariables(Particles.Animate:NiagaraFloat=&ClampFloat(Float=&Subtract_Float(
      A=&Min_Float(
       A:NiagaraFloat=$User.RiftOpen
       B:NiagaraFloat=&FloatFromCurve(FloatCurve=DataInterface() {propertyValues {Curve {keys [<at 0 0.15>,<at 0.5 1>]}}}, CurveIndex=$User.Intensity)
      )
      B=&NormalizeDistanceRange(
       Start Position:NiagaraPosition=$Engine.Emitter.SimulationPosition
       End Position:NiagaraPosition=$Particles.Position
       Distance:NiagaraFloat=3000
       Invert Normalized Range:NiagaraBool=false
      )
     ))) {isSetParameters=true, module=SetVariables_#6}
   /SetVariables(Particles.Position=&RotateVectorByQuaternion(
     Vector To Rotate=&MakeVector(
      X=$Particles.ForwardTravel
      Y=0
      Z=&Add_Float(
       A=$System.PortalRadius
       B=&Multiply_Float(A=$Particles.PartitionParticles.Normalized Index In Partition, B=1500)
      )
     )
     Quaternion=&MakeQuaternion(
      Quaternion From=0
      Axis=<1 0 0>
      Angle=&Add_Float(
       A=&SelectFloatFromArray(Direct Array Index=$Particles.PartitionParticles.Partition, Float Selection Array=DataInterface() {propertyValues {FloatData [0.2,0.5]}})
       B=&FloatFromCurve(
        FloatCurve=DataInterface() {propertyValues {Curve {
           defaultValue=1
           keys [
            {arriveTangent=5.0279, interpMode=RCIM_Cubic, leaveTangent=5.0279, tangentMode=RCTM_User, time=0, value=0}
            {arriveTangent=0.1568, interpMode=RCIM_Cubic, leaveTangent=0.1568, tangentMode=RCTM_User, time=0.5, value=1}
            {arriveTangent=0.1, interpMode=RCIM_Cubic, leaveTangent=0.1, time=0.7, value=1}
            {arriveTangent=0.2723, interpMode=RCIM_Cubic, leaveTangent=0.2723, tangentMode=RCTM_User, time=1, value=1.05}
           ]
          }}}
        CurveIndex=$Particles.PartitionParticles.Normalized Index In Partition
        Scale Curve=-0.4
       )
      )
      From=?
      To=?
      Angle Type=::NewEnumerator1
      Normalize Vectors=false
     )
    )) {isSetParameters=true, module=SetVariables_#7}
   &DynamicMaterialParameters(
    Index 0 Param 1=&FloatFromCurve(
     FloatCurve=DataInterface() {propertyValues {Curve {
        defaultValue=1
        keys [
         {arriveTangent=2.7272, interpMode=RCIM_Cubic, leaveTangent=2.7272, tangentMode=RCTM_User, time=0, value=0}
         {tangentMode=RCTM_User, time=1, value=1}
        ]
       }}}
     CurveIndex=$Particles.Animate
     Scale Curve=&FloatFromCurve(
      FloatCurve=DataInterface() {propertyValues {Curve {
         defaultValue=1
         keys [
          {arriveTangent=23.1802, interpMode=RCIM_Cubic, leaveTangent=23.1802, tangentMode=RCTM_User, time=0, value=0}
          {arriveTangentWeight=0.085, interpMode=RCIM_Cubic, leaveTangentWeight=0.1014, tangentMode=RCTM_User, time=0.1, value=1}
          {arriveTangent=-1.3131, interpMode=RCIM_Cubic, leaveTangent=-1.3131, tangentMode=RCTM_User, time=0.5126, value=0.8}
          {arriveTangent=-3.4125, interpMode=RCIM_Cubic, leaveTangent=-3.4125, tangentMode=RCTM_User, time=1, value=0}
         ]
        }}}
      CurveIndex=$Particles.PartitionParticles.Normalized Index In Partition
     )
    )
    Index 0 Param 3=&Add_Float(
     A=&Multiply_Float(A=$Particles.Age, B=1)
     B=&FloatFromCurve(
      FloatCurve=DataInterface() {propertyValues {Curve {
         defaultValue=1
         keys [{arriveTangent=2.8195, interpMode=RCIM_Cubic, leaveTangent=2.8195, tangentMode=RCTM_User, time=0, value=0}, <at 1 1 RCIM_Cubic>]
        }}}
      CurveIndex=$Particles.Animate
      Scale Curve=5
     )
    )
    Index 0 Param 4=&Waveform(
     Global Input Position=$Particles.PartitionParticles.Normalized Index In Partition
     Global Phase=&Multiply_Float(A=$Particles.Age, B=-0.8)
     "Waveform [1]"=::NewEnumerator0
     "[1] Phase Scale"=0.18
     "[1] Frequency"=6
     "[1] Amplitude Min/Max"=<-1 3>
    )
    Module.Index0Float0Write=true
    Module.Index0Float1Write=false
    Module.Index0Float2Write=true
    Module.Index0Float3Write=true
   )
   &ScaleColor(
    Scale RGB=&VectorFromFloat(Value=&Lerp_Float(A=1, B=2, Alpha=$System.RiftPulse))
    ScaleA=false
    ScaleRGB=true
   )
  ]
  rendererClasses [&NiagaraSpriteRendererProperties, &NiagaraRibbonRendererProperties]
  renderers [
   {rendererClass=&NiagaraSpriteRendererProperties}
   {rendererClass=&NiagaraRibbonRendererProperties, rendererIndex=1}
  ]
 }
 TetherMotes {
  emitterUpdateScript [
   &EmitterState()
   &SpawnParticlesFromOtherEmitter(
    Attribute Reader=DataInterface() {propertyValues {EmitterBinding {bindingMode=Other, emitterName=TetherRibbons}}}
    Spawn Rate=&Lerp_Float(A=3, B=7, Alpha=$User.Intensity)
    Spawn Rate Per Particle Cap=7
    Use Maximum Age=false
    Use Minimum Age=false
   )
  ]
  particleSpawnScript [
   &InitializeParticle(
    Color Maximum=<rgba 2.5 0.8289 1.9131 1>
    Color Minimum=<rgba 1.5 0.0156 0.7344 1>
    Color Mode=::NewEnumerator2
    Lifetime Mode=::NewEnumerator1
    Sprite Size Mode=::NewEnumerator1
    Uniform Sprite Size Max=16
    Uniform Sprite Size Min=9
    UsePositionOffset=false
    Write Material Random=true
   )
   &SampleParticlesFromOtherEmitter()
   &AddVelocity(
    Velocity=&self:GetRibbonDirection(
     Particle Attribute Reader:NiagaraDataInterfaceParticleRead=DataInterface() {propertyValues {EmitterBinding {bindingMode=Other, emitterName=TetherRibbons}}}
     Index:NiagaraInt32=&MakeIntFromFloat(Float=$Particles.SampleParticlesFromOtherEmitter.SampledIndex)
     IndexOffset:NiagaraInt32=1
     FallBackVelocity:Vector3f=&Multiply_VectorByFloat(
      Vector=&TransformVector(Vector=$System.DirectionToTree, Source Space=::World, Destination Space=::Simulation)
      Float=-500
     )
    )
    Velocity Speed Scale=&Lerp_Float(
     A=&RandomRangeFloat(Minimum=0.6, Maximum=1.2)
     B=&RandomRangeFloat(Minimum=1, Maximum=2)
     Alpha=$User.Intensity
    )
   )
   &ShapeLocation(MaskSpawnGroup=false, Override Seed=false, Sphere Radius=50, Sphere Surface Distribution=1, UseOwnerScale=true)
  ]
  particleUpdateScript [
   &ParticleState()
   &ScaleSpriteSize(
    Uniform Curve Scale=&FloatFromCurve(FloatCurve=DataInterface() {propertyValues {Curve {keys [<at 6000 0.5>,<at 9000 1>,<at 15000 3>]}}}, CurveIndex=$Engine.Owner.LODDistance)
    Uniform Curve Sprite Scale=DataInterface() {propertyValues {Curve {
       defaultValue=0
       keys [
        {arriveTangent=5.1504, interpMode=RCIM_Cubic, leaveTangent=5.1504, tangentMode=RCTM_User, time=0, value=0}
        {interpMode=RCIM_Cubic, tangentMode=RCTM_User, time=0.5, value=1}
        {arriveTangent=-4.4997, tangentMode=RCTM_User, time=1, value=0}
       ]
      }}}
   )
   &VectorNoiseForce(Force Amount=2000, Module.Use Random Noise Time Interval=true, Random Noise Time Interval=0.6)
   &Drag(Use Linear Drag=true, Use Rotational Drag=false)
   &SolveForcesAndVelocity(Clamp Velocity=false, Limit Acceleration=false)
   &ScaleSpriteSizeBySpeed(
    Max Scale Factor=<0.7 5>
    Sample Scale Factor By Curve=true
    Scale Factor Curve=DataInterface() {propertyValues {Curve {keys [<at 0 0 RCIM_Cubic>,<at 1 1>]}}}
    Velocity Threshold=1500
   )
   &ScaleColor(
    Scale RGB=&VectorFromFloat(Value=&Lerp_Float(A=1, B=5, Alpha=$System.RiftPulse))
    ScaleA=false
    ScaleRGB=true
   )
  ]
  rendererClasses [&NiagaraSpriteRendererProperties]
  renderers [{rendererClass=&NiagaraSpriteRendererProperties}]
 }
 TetherRibbons {
  emitterUpdateScript [
   &EmitterState(
    Inactive Response=::NewEnumerator1
    Life Cycle Mode=::NewEnumerator1
    Loop Behavior=::NewEnumerator1
    Loop Duration Mode=::NewEnumerator1
    UseLoopDelay=false
   )
   &SpawnBurst_Instantaneous(Spawn Count=24, Use Spawn Probability=false, UseLoopCountLimit=false)
  ]
  particleSpawnScript [
   &PartitionParticles(
    Normalized Index In Partition Includes One=true
    Particles Per Partition=12
    Partition Mode:ENiagara_PartitionMode_v2=::NewEnumerator1
    Write Partition Index to Ribbon Link Order:ENiagara_AttributeSamplingApplyOutput=::NewEnumerator0
    Write Partition to Ribbon ID=::NewEnumerator0
   )
   &InitializeParticle(
    Color=&LerpLinearColors(
     StartColor=<rgba 2 0.1042 1.0222 1>
     EndColor=<rgba 0.6518 0.224 1 0.25>
     LerpFactor=&FloatFromCurve(
      FloatCurve=DataInterface() {propertyValues {Curve {
         defaultValue=0
         keys [
          {arriveTangent=2.87, interpMode=RCIM_Cubic, leaveTangent=2.87, tangentMode=RCTM_User, time=0, value=0}
          {interpMode=RCIM_Cubic, tangentMode=RCTM_User, time=0.5, value=1}
          {arriveTangent=-2.9896, tangentMode=RCTM_User, time=1, value=0}
         ]
        }}}
      CurveIndex=$Particles.PartitionParticles.Normalized Index In Partition
     )
    )
    Lifetime=2
    Ribbon Facing Mode=::NewEnumerator1
    Ribbon Width=&FloatFromCurve(
     FloatCurve=DataInterface() {propertyValues {Curve {keys [
         <at 0 2.4 RCIM_Cubic>
         {arriveTangentWeight=0.1374, interpMode=RCIM_Cubic, leaveTangentWeight=0.2062, tangentMode=RCTM_User, time=0.125, value=1}
         {arriveTangentWeight=0.1202, interpMode=RCIM_Cubic, leaveTangentWeight=0.2404, tangentMode=RCTM_User, time=0.4, value=1}
         {arriveTangent=-1.4976, tangentMode=RCTM_User, time=1, value=0.4}
        ]}}}
     CurveIndex=$Particles.PartitionParticles.Normalized Index In Partition
     Scale Curve=300
    )
    Ribbon Width Mode=::NewEnumerator1
    UsePositionOffset=false
    Write Material Random=true
   )
  ]
  particleUpdateScript [
   &ParticleState(Kill Particles When Lifetime Has Elapsed=false)
   /SetVariables(Particles.Position=&AddVectorToPosition(
     Position=&LerpPosition(A=$System.LocalTreePosition, B=<1500 100 -1650>, Alpha=$Particles.PartitionParticles.Normalized Index In Partition)
     Vector=&MaskVector(
      Mask Source=1
      Mask Parameter=$Particles.PartitionParticles.Partition
      Mask=0
      Masked Vector=&VectorFromCurve(
       VectorCurve=DataInterface() {propertyValues {
         XCurve {keys [
           {interpMode=RCIM_Cubic, leaveTangentWeight=0.0335, tangentMode=RCTM_User, time=0, value=0}
           {arriveTangent=0.6894, interpMode=RCIM_Cubic, leaveTangent=0.6894, time=0.1, value=-0.15}
           {arriveTangent=-0.7546, interpMode=RCIM_Cubic, leaveTangent=-0.7546, tangentMode=RCTM_User, time=0.2699, value=0.1861}
           <at 1 0>
          ]}
         YCurve {keys [
           {arriveTangent=4.9812, interpMode=RCIM_Cubic, time=0, value=0}
           {arriveTangent=-5.5, interpMode=RCIM_Cubic, leaveTangent=-5.5, tangentMode=RCTM_User, time=0.0734, value=0.374}
           {arriveTangent=0.352, interpMode=RCIM_Cubic, leaveTangent=0.352, tangentMode=RCTM_User, time=0.3729, value=-0.3754}
           {arriveTangent=-2.0083, interpMode=RCIM_Cubic, leaveTangent=-2.0083, tangentMode=RCTM_User, time=0.8, value=0.3569}
           <at 1 0 RCIM_Cubic>
          ]}
         ZCurve {keys [
           {arriveTangent=-13.0056, interpMode=RCIM_Cubic, leaveTangent=-13.0056, tangentMode=RCTM_User, time=0, value=1}
           {arriveTangent=-3.1327, interpMode=RCIM_Cubic, leaveTangent=-3.1327, time=0.3399, value=-0.3996}
           {arriveTangent=0.254, interpMode=RCIM_Cubic, leaveTangent=0.254, time=0.6384, value=-1}
           {arriveTangent=2.7658, interpMode=RCIM_Cubic, leaveTangent=2.7658, time=0.8536, value=-0.2691}
           {interpMode=RCIM_Cubic, leaveTangent=0.5706, time=1, value=0}
          ]}
        }}
       CurveIndex=$Particles.PartitionParticles.Normalized Index In Partition
       Scale Curve=<500 500 600>
      )
      Passthrough Vector=&VectorFromCurve(
       VectorCurve=DataInterface() {propertyValues {
         XCurve {keys [
           {arriveTangent=-3.0139, interpMode=RCIM_Cubic, leaveTangent=-3.0139, tangentMode=RCTM_User, time=0, value=0}
           {arriveTangent=0.9424, interpMode=RCIM_Cubic, leaveTangent=0.9424, tangentMode=RCTM_User, time=0.1535, value=-0.0617}
           {arriveTangent=-0.7546, interpMode=RCIM_Cubic, leaveTangent=-0.7546, tangentMode=RCTM_User, time=0.5393, value=0.0203}
           {interpMode=RCIM_Cubic, leaveTangent=0.2713, time=1, value=0}
          ]}
         YCurve {keys [
           {arriveTangent=-11.266, interpMode=RCIM_Cubic, leaveTangent=-11.266, tangentMode=RCTM_User, time=0, value=0}
           {arriveTangent=0.1276, interpMode=RCIM_Cubic, leaveTangent=0.1276, tangentMode=RCTM_User, time=0.1143, value=-0.4756}
           {arriveTangent=-1.2086, interpMode=RCIM_Cubic, leaveTangent=-1.2086, tangentMode=RCTM_User, time=0.4121, value=0.5}
           {arriveTangent=-0.6758, interpMode=RCIM_Cubic, leaveTangent=-0.6758, time=0.6967, value=-0.1881}
           {arriveTangent=0.6168, interpMode=RCIM_Cubic, leaveTangent=0.6168, time=0.8769, value=0.1859}
           {arriveTangent=-1.9289, interpMode=RCIM_Cubic, leaveTangent=-1.9289, tangentMode=RCTM_User, time=1.0016, value=0}
          ]}
         ZCurve {keys [
           <at 0 0 RCIM_Cubic>
           {arriveTangent=-2.317, interpMode=RCIM_Cubic, leaveTangent=-2.317, tangentMode=RCTM_User, time=0.1869, value=-0.6658}
           {arriveTangent=0.1607, interpMode=RCIM_Cubic, leaveTangent=0.1607, tangentMode=RCTM_User, time=0.6453, value=-1}
           {arriveTangent=5.1933, interpMode=RCIM_Cubic, leaveTangent=5.1933, tangentMode=RCTM_User, time=1, value=0}
          ]}
        }}
       CurveIndex=$Particles.PartitionParticles.Normalized Index In Partition
       Scale Curve=<500 500 400>
      )
     )
    )) {isSetParameters=true, module=SetVariables_#7}
   &DynamicMaterialParameters(
    Index 0 Param 1=&ClampFloat(
     Float=&FloatFromCurve(
      FloatCurve=DataInterface() {propertyValues {Curve {
         defaultValue=1
         keys [
          {arriveTangent=-5.3116, interpMode=RCIM_Cubic, leaveTangent=-5.3116, leaveTangentWeight=0.1667, tangentMode=RCTM_User, time=0, value=0.6}
          {arriveTangentWeight=0.0667, interpMode=RCIM_Cubic, leaveTangentWeight=0.1667, tangentMode=RCTM_User, time=0.2, value=0}
          {arriveTangentWeight=0.2668, interpMode=RCIM_Cubic, leaveTangentWeight=0.1601, tangentMode=RCTM_User, time=0.7, value=0}
          {arriveTangent=4.1181, interpMode=RCIM_Cubic, leaveTangent=4.1181, tangentMode=RCTM_User, time=1, value=0.5}
         ]
        }}}
      CurveIndex=$Particles.PartitionParticles.Normalized Index In Partition
     )
     Min=&Lerp_Float(A=1, B=0, Alpha=$User.RiftOpen)
    )
    Index 0 Param 2=&MaskFloat(Mask Source=0, Mask Parameter=?, Mask=$Particles.PartitionParticles.Partition, Masked Float=6, Passthrough Float=4)
    Index 0 Param 3=&MaskFloat(Mask Source=0, Mask Parameter=?, Mask=$Particles.PartitionParticles.Partition, Masked Float=0, Passthrough Float=0.7)
    Index 1 Param 3=$User.PulseTravel
    Index 1 Param 4=$User.PulseColorShift
    Module.Index0Float0Write=true
    Module.Index0Float1Write=true
    Module.Index0Float2Write=true
    Module.Index0Float3Write=false
    Module.Index1Float0Write=false
    Module.Index1Float1Write=false
    Module.Index1Float2Write=true
    Module.Index1Float3Write=true
    Write Parameter Index 1=::NewEnumerator1
   )
   &ScaleRibbonWidth(Ribbon Width Scale=&FloatFromCurve(
     FloatCurve=DataInterface() {propertyValues {Curve {keys [{arriveTangent=2.5115, interpMode=RCIM_Cubic, leaveTangent=2.5115, tangentMode=RCTM_User, time=0, value=0}, <at 1 1>]}}}
     CurveIndex=$User.RiftOpen
     Scale Curve=&Lerp_Float(A=0.5, B=1, Alpha=$User.Intensity)
    ))
   &ScaleColor(Scale Alpha=&Lerp_Float(A=0.5, B=1, Alpha=$User.Intensity), ScaleA=true, ScaleRGB=false)
  ]
  rendererClasses [&NiagaraSpriteRendererProperties, &NiagaraRibbonRendererProperties]
  renderers [
   {rendererClass=&NiagaraSpriteRendererProperties}
   {rendererClass=&NiagaraRibbonRendererProperties, rendererIndex=1}
  ]
 }
}
summary {
 systemName=NS_LimeRock_RealityTear_Portal
 emitters [
  {emitterName=MainRift, rendererClasses [&NiagaraSpriteRendererProperties, &NiagaraRibbonRendererProperties]}
  {emitterName=RiftMotes, rendererClasses [&NiagaraSpriteRendererProperties]}
  {emitterName=Portal, rendererClasses [&NiagaraMeshRendererProperties]}
  {
   emitterName=TetherRibbons
   rendererClasses [&NiagaraSpriteRendererProperties, &NiagaraRibbonRendererProperties]
  }
  {bEnabled=false, emitterName=InteriorSmoke, rendererClasses [&NiagaraSpriteRendererProperties]}
  {emitterName=SideRifts, rendererClasses [&NiagaraSpriteRendererProperties, &NiagaraRibbonRendererProperties]}
  {
   emitterName=SheddingRibbons
   rendererClasses [&NiagaraSpriteRendererProperties, &NiagaraRibbonRendererProperties]
  }
  {emitterName=TetherMotes, rendererClasses [&NiagaraSpriteRendererProperties]}
  {
   emitterName=CenterFlare
   rendererClasses [&NiagaraSpriteRendererProperties, &NiagaraComponentRendererProperties]
  }
 ]
 userVariables [
  User.Intensity {type=NiagaraFloat, defaultValue=0, description=""}
  User.PulseColorShift {type=NiagaraFloat, defaultValue=0, description=""}
  User.PulseTravel {type=NiagaraFloat, defaultValue=0, description=""}
  User.RiftJitter {type=NiagaraFloat, defaultValue=0, description=""}
  User.RiftOpen {type=NiagaraFloat, defaultValue=1, description=""}
 ]
}
